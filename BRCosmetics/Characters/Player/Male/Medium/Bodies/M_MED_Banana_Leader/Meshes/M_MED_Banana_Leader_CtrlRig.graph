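# Control Rig for the M_MED_Banana_Leader body that procedurally adjusts the left shoulder pad as the arm moves.
# Clavicle and upper-arm Euler angles are remapped to clamped weights that offset the pad's parent bone, with the max-combined arm weight also modifying feather bones.
# A separate chain weights another pad offset by elbow-to-target distance through a Kalman filter.
# Older driver-bone, aim and constraint nodes sit disconnected; right-side bones are untouched.
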ControlRigBlueprint /BRCosmetics/Characters/Player/Male/Medium/Bodies/M_MED_Banana_Leader/Meshes/M_MED_Banana_Leader_CtrlRig
graphs [
 RigVMFunctionLibrary {links [], nodes []}
 RigVMModel {nodes [
   n1 = &RigUnit_GetRelativeBoneTransform(Bone=upperarm_l, Space=clavicle_l)
   /Script/ControlRig.RigUnit_ModifyBoneTransforms(
    BoneToModify=((Bone="dyn_driver_shldrpad_l",Transform=(Rotation=(X=0,Y=0,Z=0,W=1),Translation=(X=3,Y=0,Z=0.5),Scale3D=(X=1,Y=1,Z=1))))
    ExecutePin=()
    Mode=AdditiveLocal
    WeightMaximum=1
    Weight=&RigVMFunction_MathFloatRemap(
     SourceMaximum=80
     SourceMinimum=0
     TargetMinimum=0
     bClamp=true
     Value=&RigVMFunction_MathQuaternionToEuler(RotationOrder=ZXY, Value=@n1.Transform.Rotation).Result.Z
    ).Result
    BoneToModify.0.Transform.Rotation=&RigVMFunction_MathQuaternionFromEuler(Euler=(X=15,Y=10,Z=11)).Result
   ) {is=struct RigVMUnitNode}
   RigUnit_GetBoneTransform = &RigUnit_GetBoneTransform(Bone=lowerarm_l)
   n2 = &RigVMFunction_MathVectorDistance(A=@RigUnit_GetBoneTransform.Transform.Translation, B=&RigUnit_GetBoneTransform(Bone=dyn_arm_up).Transform.Translation)
   > &RigUnit_ModifyBoneTransforms(
    BoneToModify=((Bone="dyn_driver_shldrpad_l",Transform=(Rotation=(X=0,Y=0,Z=0,W=1),Translation=(X=5,Y=0,Z=0),Scale3D=(X=1,Y=1,Z=1))))
    Mode=AdditiveLocal
    WeightMaximum=1
    Weight=&RigVMFunction_MathFloatRemap(SourceMaximum=11, SourceMinimum=25, bClamp=true, Value=@n2.Result).Result
   )
   n3 = &RigUnit_BeginExecution()
   RigUnit_GetTransform = &RigUnit_GetTransform(Item=(Type=Bone,Name="root"))
   /Script/ControlRig.RigUnit_DebugTransformMutableItemSpace(
    Color=(R=1,G=0,B=0,A=1)
    DebugDrawSettings=(DepthPriority=SDPG_Foreground,Lifetime=-1)
    Mode=Axes
    Scale=30
    Space=(Type=Bone)
    Thickness=1
    WorldOffset=(Rotation=(X=0,Y=0,Z=0,W=1),Translation=(X=0,Y=0,Z=0),Scale3D=(X=1,Y=1,Z=1))
    bEnabled=True
    ExecutePin=@n3.ExecutePin
    Transform=@RigUnit_GetTransform.Transform
   ) {is=struct RigVMUnitNode}
   n4 = &RigUnit_GetRelativeTransformForItem(Child=(Type=Bone,Name="clavicle_l"), Parent=(Type=Bone,Name="spine_05"))
   > &RigUnit_ModifyTransforms(
    ItemToModify=((Item=(Type=Bone,Name="dyn_parent_shoulderpad_l"),Transform=(Rotation=(X=0,Y=0,Z=0,W=1),Translation=(X=0,Y=0,Z=-1.5),Scale3D=(X=1,Y=1,Z=1))))
    Weight=&RigVMFunction_MathFloatRemap(
     SourceMaximum=50
     SourceMinimum=82
     bClamp=true
     Value=&RigVMFunction_MathQuaternionToEuler(RotationOrder=ZYX, Value=@n4.RelativeTransform.Rotation).Result.Y
    ).Result
    ItemToModify.0.Transform.Rotation=&RigVMFunction_MathQuaternionFromEuler(Euler=(X=0,Y=40,Z=0), RotationOrder=XYZ).Result
   )
   n5 = &RigUnit_GetRelativeTransformForItem(Child=(Type=Bone,Name="upperarm_l"), Parent=(Type=Bone,Name="clavicle_l"))
   n6 = &RigVMFunction_MathQuaternionToEuler(RotationOrder=XYZ, Value=@n5.RelativeTransform.Rotation)
   n7 = &RigVMFunction_MathQuaternionToEuler(RotationOrder=XYZ, Value=@n5.RelativeTransform.Rotation)
   n8 = &RigUnit_GetRelativeTransformForItem(Child=(Type=Bone,Name="upperarm_l"), Parent=(Type=Bone,Name="clavicle_l"))
   n9 = /Script/RigVM.RigVMFunction_MathFloatMax(
    A=&RigVMFunction_MathFloatMax(
     B=&RigVMFunction_MathFloatRemap(SourceMaximum=9, SourceMinimum=70, TargetMaximum=0.8, TargetMinimum=0, bClamp=true, Value=@n6.Result.Y).Result
     A=&RigVMFunction_MathFloatRemap(SourceMaximum=-8, SourceMinimum=70, TargetMinimum=0, bClamp=true, Value=@n7.Result.Y).Result
    ) {template="Maximum::Execute(in A,in B,out Result)"}.Result
    B=&RigVMFunction_MathFloatRemap(
     SourceMaximum=100
     SourceMinimum=45
     TargetMinimum=0
     bClamp=true
     Value=&RigVMFunction_MathQuaternionToEuler(RotationOrder=ZXY, Value=@n8.RelativeTransform.Rotation).Result.Z
    ).Result
   ) {is=struct RigVMUnitNode, template="Maximum::Execute(in A,in B,out Result)"}
   > &RigUnit_ModifyTransforms(
    ItemToModify=((Item=(Type=Bone,Name="dyn_parent_shoulderpad_l"),Transform=(Rotation=(X=0,Y=0,Z=0,W=1),Translation=(X=0.5,Y=-2.9,Z=1),Scale3D=(X=1,Y=1,Z=1))))
    ItemToModify.0.Transform.Rotation=&RigVMFunction_MathQuaternionFromEuler(Euler=(X=40,Y=-20,Z=-10), RotationOrder=YXZ).Result
    Weight=@n9.Result
   )
   > &RigUnit_ModifyTransforms(
    ItemToModify="((Item=(Type=Bone,Name=\"dyn_shldr_feather_a_l\"),Transform=(Rotation=(X=0,Y=0,Z=0,W=1),Translation=(X=0,Y=0,Z=0),Scale3D=(X=0.7,Y=0.7,Z=0.7))),(Item=(Type=Bone,Name=\"dyn_shldr_feather_b_l\"),Transform=(Rotation=(X=0,Y=0,Z=0,W=1),Translation=(X=0,Y=0,Z=0),Sca ...(truncated)"
    Weight=@n9.Result
   )
   n10 = &RigUnit_BeginExecution()
   n11 = &RigVMFunction_MathVectorDistance(
    B=&RigUnit_GetTransform(Item=(Type=Bone,Name="target_arm_up_l")).Transform.Translation
    A=&RigUnit_GetTransform(Item=(Type=Bone,Name="lowerarm_l")).Transform.Translation
   )
   &RigUnit_ModifyTransforms(
    ItemToModify=((Item=(Type=Bone,Name="dyn_parent_shoulderpad_l"),Transform=(Rotation=(X=0,Y=0,Z=0,W=1),Translation=(X=2,Y=-2,Z=1),Scale3D=(X=1,Y=1,Z=1))))
    Weight=&RigVMFunction_KalmanFloat(BufferSize=1, Value=&RigVMFunction_MathFloatRemap(SourceMaximum=7, SourceMinimum=15, bClamp=true, Value=@n11.Result).Result) {template="KalmanFilter::Execute(in Value,in BufferSize,out Result)"}.Result
    ItemToModify.0.Transform.Rotation=&RigVMFunction_MathQuaternionFromEuler(Euler=(X=40,Y=-20,Z=-5), RotationOrder=YXZ).Result
    ExecutePin=@n10.ExecutePin
   )
   n12 = &RigUnit_GetRelativeTransformForItem(Child=(Type=Bone,Name="lowerarm_l"), Parent=(Type=Bone,Name="dyn_shoulderpad_base_l"), bChildInitial=true, bParentInitial=true)
   /Script/ControlRig.RigUnit_AimItem(
    DebugSettings=(Scale=10,WorldOffset=(Rotation=(X=0,Y=0,Z=0,W=1),Translation=(X=0,Y=0,Z=0),Scale3D=(X=1,Y=1,Z=1)))
    ExecutePin=()
    Item=(Type=Bone,Name="dyn_shoulderpad_base_l")
    Primary=(Weight=1,Axis=(X=1,Y=0,Z=0),Target=(X=0,Y=0,Z=0),Kind=Location,Space=(Type=Bone,Name="lowerarm_l"))
    Secondary=(Weight=0,Axis=(X=0,Y=0,Z=1),Target=(X=1,Y=0,Z=0),Kind=Location,Space=(Type=Bone))
    Weight=1
    Primary.Axis=@n12.RelativeTransform.Translation
   ) {is=struct RigVMUnitNode}
   &RigUnit_DebugTransformMutableItemSpace(
    Color=(R=1,G=0,B=0,A=1)
    DebugDrawSettings=(DepthPriority=SDPG_Foreground,Lifetime=-1)
    ExecutePin=()
    Mode=Axes
    Scale=10
    Space=(Type=Bone)
    Transform=(Rotation=(X=0,Y=0,Z=0,W=1),Translation=(X=0,Y=0,Z=0),Scale3D=(X=1,Y=1,Z=1))
    WorldOffset=(Rotation=(X=0,Y=0,Z=0,W=1),Translation=(X=0,Y=0,Z=0),Scale3D=(X=1,Y=1,Z=1))
    bEnabled=True
   )
   /Script/RigVM.RigVMFunction_MathFloatClamp(Maximum=11, Minimum=0, Value=@n2.Result) {is=struct RigVMUnitNode, template="Clamp::Execute(in Value,in Minimum,in Maximum,out Result)"}
   &RigVMFunction_MathFloatRemap(
    SourceMaximum=-65
    SourceMinimum=0
    TargetMinimum=0
    bClamp=true
    Value=&RigVMFunction_MathQuaternionToEuler(RotationOrder=ZYX, Value=&RigUnit_GetRelativeBoneTransform(Bone=upperarm_l, Space=clavicle_l).Transform.Rotation).Result.Y
   )
   &RigVMFunction_MathFloatRemap(
    SourceMaximum=6
    SourceMinimum=23
    bClamp=true
    Value=&RigVMFunction_MathVectorDistance(
     A=&RigUnit_GetTransform(Item=(Type=Bone,Name="lowerarm_l")).Transform.Translation
     B=&RigUnit_GetTransform(Item=(Type=Bone,Name="target_arm_fwd_l")).Transform.Translation
    ).Result
   )
   RigUnit_BeginExecution = &RigUnit_BeginExecution()
   /Script/ControlRig.RigUnit_TransformConstraintPerItem(
    BaseItem=(Type=Bone)
    BaseTransform=(Rotation=(X=0,Y=0,Z=0,W=1),Translation=(X=0,Y=0,Z=0),Scale3D=(X=1,Y=1,Z=1))
    BaseTransformSpace=GlobalSpace
    ExecutePin=()
    Item=(Type=Bone,Name="target_lowerarm_fwd")
    Targets=((Transform=(Rotation=(X=0,Y=0,Z=0,W=1),Translation=(X=0,Y=0,Z=0),Scale3D=(X=1,Y=1,Z=1)),Weight=1,bMaintainOffset=True,Filter=(TranslationFilter=(bX=True,bY=True,bZ=True),RotationFilter=(bX=True,bY=True,bZ=True),ScaleFilter=(bX=True,bY=True,bZ=True))))
    bUseInitialTransforms=True
    Targets.0.Transform=&RigUnit_GetTransform(Item=(Type=Bone,Name="spine_05")).Transform
   ) {is=struct RigVMUnitNode}
  ]}
]
hierarchy [
 root {loc=<0 0 0>, rot=<0 0 0 1>}
 pelvis {parent=root, loc=<0 -0.7071 93.1191>, rot=<0.0127 0.707 -0.0127 -0.707>}
 spine_01 {parent=pelvis, loc=<6.7494 -0.6388 0>, rot=<0 0 0.034 -0.9994>}
 spine_02 {parent=spine_01, loc=<7.2518 -0.035 0>, rot=<0 0 0.0238 -0.9997>}
 spine_03 {parent=spine_02, loc=<6.0104 0 0>, rot=<0 0 0.0149 -0.9999>}
 spine_04 {parent=spine_03, loc=<8.1024 0 0>, rot=<0 0 -0.0132 -0.9999>}
 spine_05 {parent=spine_04, loc=<10.0146 0 0>, rot=<0 0 -0.0015 -1>}
 clavicle_l {parent=spine_05, loc=<4.3575 -0.1466 -2.3223>, rot=<0.0465 -0.7289 -0.009 -0.6829>}
 upperarm_l {parent=clavicle_l, loc=<15.5111 0 0>, rot=<0.0226 -0.3234 -0.0088 -0.9459>}
 lowerarm_l {parent=upperarm_l, loc=<26.6514 0 0>, rot=<0.0391 -0.0099 -0.2453 -0.9686>}
 hand_l {parent=lowerarm_l, loc=<22.3491 0 0>, rot=<-0.7081 -0.0507 0.0975 -0.6975>}
 index_metacarpal_l {parent=hand_l, loc=<1.6535 0.0183 -2.96>, rot=<0.0158 -0.0124 -0.0173 -0.9996>}
 index_01_l {parent=index_metacarpal_l, loc=<8.3008 0.0357 -0.0524>, rot=<-0.0044 0.0485 0.0924 -0.9945>}
 index_02_l {parent=index_01_l, loc=<3.3587 -0.3004 0.0452>, rot=<-0.002 0.0027 0.3225 -0.9466>}
 index_03_l {parent=index_02_l, loc=<2.5578 -0.123 -0.0252>, rot=<-0.002 0.0067 0.2558 -0.9667>}
 middle_metacarpal_l {parent=hand_l, loc=<2.1856 -0.1725 -0.5984>, rot=<0.006 0.0168 -0.0329 -0.9993>}
 middle_01_l {parent=middle_metacarpal_l, loc=<7.3275 -0.2447 0.0129>, rot=<-0.0208 0.0042 0.0858 -0.9961>}
 middle_02_l {parent=middle_01_l, loc=<4.0123 0.1401 0.1175>, rot=<-0.0218 0.0258 0.3512 -0.9357>}
 middle_03_l {parent=middle_02_l, loc=<2.6077 0.0534 0.1373>, rot=<0.0082 0.0188 0.1708 -0.9851>}
 pinky_metacarpal_l {parent=hand_l, loc=<1.7616 -0.9753 3.3814>, rot=<-0.0397 0.0814 -0.069 -0.9935>}
 pinky_01_l {parent=pinky_metacarpal_l, loc=<7.1709 -0.4831 -0.1094>, rot=<-0.0654 -0.0035 0.2242 -0.9723>}
 pinky_02_l {parent=pinky_01_l, loc=<2.8274 0.0893 -0.1043>, rot=<0.0037 0.0011 0.2854 -0.9584>}
 pinky_03_l {parent=pinky_02_l, loc=<2.2046 0.0307 -0.0078>, rot=<-0.0002 0.0098 0.1445 -0.9895>}
 ring_metacarpal_l {parent=hand_l, loc=<2.2074 -0.3625 1.3638>, rot=<-0.0284 0.0722 -0.0458 -0.9959>}
 ring_01_l {parent=ring_metacarpal_l, loc=<7.0295 -0.6197 -0.1446>, rot=<-0.0489 -0.0252 0.1892 -0.9804>}
 ring_02_l {parent=ring_01_l, loc=<3.6174 0.4647 -0.1027>, rot=<0 0 0.2806 -0.9598>}
 ring_03_l {parent=ring_02_l, loc=<2.4152 0.114 0.0199>, rot=<0 0 0.1189 -0.9929>}
 thumb_01_l {parent=hand_l, loc=<1.3118 -0.3191 -2.459>, rot=<0.6546 -0.4042 0.0664 -0.6354>}
 thumb_02_l {parent=thumb_01_l, loc=<5.9336 -0.0448 0.0401>, rot=<-0.0037 0.0213 0.298 -0.9543>}
 thumb_03_l {parent=thumb_02_l, loc=<2.2223 0.3682 -0.0907>, rot=<0.0118 -0.013 0.0066 -0.9998>}
 lowerarm_twist_01_l {parent=lowerarm_l, loc=<17.8549 0 0>, rot=<0 0 0 -1>}
 lowerarm_twist_02_l {parent=lowerarm_l, loc=<9.6199 0 0>, rot=<0 0 0 -1>}
 dyn_volume_base_lowerarm_l {parent=lowerarm_l, loc=<0 -4 0>, rot=<0 0 0 -1>}
 dyn_volume_lowerarm_l {parent=dyn_volume_base_lowerarm_l, loc=<0 -0.0033 0>, rot=<0 0 0 -1>, scale=<0.9993 1 1>}
 upperarm_twist_01_l {parent=upperarm_l, loc=<0.0041 0 0>, rot=<0 0 0 -1>}
 upperarm_twist_02_l {parent=upperarm_l, loc=<13.7941 0 0>, rot=<0 0 0 -1>}
 dyn_arm_measure_l {parent=clavicle_l, loc=<36.5819 0.0555 -16.3191>, rot=<0.0008 0.0336 -0.0252 -0.9991>}
 dyn_parent_shoulderpad_l {parent=clavicle_l, loc=<15.9168 -0.7389 0.4767>, rot=<0.0008 0.0336 -0.0252 -0.9991>}
 dyn_shoulderpad_base_l {parent=dyn_parent_shoulderpad_l, loc=<0 0.769 0>, rot=<0.0134 -0.3555 0.014 -0.9345>}
 dyn_shoulderpad_l {parent=dyn_shoulderpad_base_l, loc=<0 0 0>, rot=<0 0 0 -1>}
 dyn_shldr_feather_a_l {parent=dyn_shoulderpad_l, loc=<3.7562 9.04 1.3151>, rot=<0.4722 -0.1834 -0.2349 -0.8296>}
 dyn_shldr_feather_b_l {parent=dyn_shoulderpad_l, loc=<4.7803 7.5912 3.7369>, rot=<0.4508 -0.1585 -0.1663 -0.8626>}
 dyn_shldr_feather_c_l {parent=dyn_shoulderpad_l, loc=<5.5703 5.686 6.0375>, rot=<0.3963 -0.1323 -0.0934 -0.9037>}
 dyn_shldr_feather_d_l {parent=dyn_shoulderpad_l, loc=<4.8973 3.1655 8.3933>, rot=<0.2322 -0.0921 -0.0347 -0.9677>}
 dyn_shldr_feather_e_l {parent=dyn_shoulderpad_l, loc=<4.2884 -0.318 9.3417>, rot=<-0.0094 -0.0702 -0.0111 -0.9974>}
 dyn_shldr_feather_f_l {parent=dyn_shoulderpad_l, loc=<4.4991 -3.5052 8.2767>, rot=<-0.2503 -0.0802 0.0426 -0.9639>}
 dyn_shldr_feather_g_l {parent=dyn_shoulderpad_l, loc=<5.8642 -6.3468 5.9011>, rot=<-0.3815 -0.1392 0.065 -0.9115>}
 dyn_shldr_feather_h_l {parent=dyn_shoulderpad_l, loc=<5.3898 -8.1136 3.322>, rot=<-0.4441 -0.1712 0.1333 -0.8693>}
 dyn_shldr_feather_i_l {parent=dyn_shoulderpad_l, loc=<4.4016 -9.5222 0.8809>, rot=<-0.4698 -0.2088 0.1914 -0.8361>}
 clavicle_r {parent=spine_05, loc=<4.3575 -0.1466 2.3223>, rot=<0.7289 0.0465 -0.6829 0.009>}
 upperarm_r {parent=clavicle_r, loc=<-15.5111 0 0>, rot=<0.0226 -0.3234 -0.0088 -0.9459>}
 lowerarm_r {parent=upperarm_r, loc=<-26.6515 0 0>, rot=<0.0391 -0.0099 -0.2453 -0.9686>}
 hand_r {parent=lowerarm_r, loc=<-22.3492 0 0>, rot=<-0.7081 -0.0507 0.0975 -0.6975>}
 index_metacarpal_r {parent=hand_r, loc=<-1.6534 -0.0181 2.9601>, rot=<0.0158 -0.0124 -0.0173 -0.9996>}
 index_01_r {parent=index_metacarpal_r, loc=<-8.3008 -0.0357 0.0524>, rot=<-0.0044 0.0485 0.0924 -0.9945>}
 index_02_r {parent=index_01_r, loc=<-3.3589 0.3003 -0.0453>, rot=<-0.002 0.0027 0.3225 -0.9466>}
 index_03_r {parent=index_02_r, loc=<-2.5578 0.1231 0.0251>, rot=<-0.002 0.0067 0.2558 -0.9667>}
 middle_metacarpal_r {parent=hand_r, loc=<-2.1856 0.1726 0.5984>, rot=<0.006 0.0168 -0.0329 -0.9993>}
 middle_01_r {parent=middle_metacarpal_r, loc=<-7.3273 0.2447 -0.0129>, rot=<-0.0208 0.0042 0.0858 -0.9961>}
 middle_02_r {parent=middle_01_r, loc=<-4.0142 -0.1411 -0.07>, rot=<-0.0217 0.0259 0.3512 -0.9357>}
 middle_03_r {parent=middle_02_r, loc=<-2.6074 -0.0534 -0.1312>, rot=<0.0083 0.0188 0.1708 -0.9851>}
 pinky_metacarpal_r {parent=hand_r, loc=<-1.7617 0.9753 -3.3814>, rot=<-0.0397 0.0814 -0.069 -0.9935>}
 pinky_01_r {parent=pinky_metacarpal_r, loc=<-7.1707 0.4832 0.1094>, rot=<-0.0654 -0.0035 0.2242 -0.9723>}
 pinky_02_r {parent=pinky_01_r, loc=<-2.8274 -0.0892 0.1043>, rot=<0.0037 0.0011 0.2854 -0.9584>}
 pinky_03_r {parent=pinky_02_r, loc=<-2.2045 -0.031 0.0078>, rot=<-0.0002 0.0098 0.1445 -0.9895>}
 ring_metacarpal_r {parent=hand_r, loc=<-2.2073 0.3627 -1.3638>, rot=<-0.0284 0.0722 -0.0458 -0.9959>}
 ring_01_r {parent=ring_metacarpal_r, loc=<-7.0297 0.6197 0.1446>, rot=<-0.0489 -0.0252 0.1892 -0.9804>}
 ring_02_r {parent=ring_01_r, loc=<-3.6171 -0.4648 0.1027>, rot=<0 0 0.2806 -0.9598>}
 ring_03_r {parent=ring_02_r, loc=<-2.4153 -0.1141 -0.0199>, rot=<0 0 0.1189 -0.9929>}
 thumb_01_r {parent=hand_r, loc=<-1.3119 0.3192 2.459>, rot=<0.6546 -0.4042 0.0664 -0.6354>}
 thumb_02_r {parent=thumb_01_r, loc=<-5.9336 0.0447 -0.0401>, rot=<-0.0037 0.0213 0.298 -0.9543>}
 thumb_03_r {parent=thumb_02_r, loc=<-2.2223 -0.3682 0.0907>, rot=<0.0118 -0.013 0.0066 -0.9998>}
 lowerarm_twist_01_r {parent=lowerarm_r, loc=<-17.855 0 0>, rot=<0 0 0 1>}
 lowerarm_twist_02_r {parent=lowerarm_r, loc=<-9.6199 0 0>, rot=<0 0 0 1>}
 dyn_volume_base_lowerarm_r {parent=lowerarm_r, loc=<0 4 0>, rot=<0 0 0 1>}
 dyn_volume_lowerarm_r {parent=dyn_volume_base_lowerarm_r, loc=<0 0.0033 0>, rot=<0 0 0 1>, scale=<0.9993 1 1>}
 upperarm_twist_01_r {parent=upperarm_r, loc=<-0.0041 0 0>, rot=<0 0 0 1>}
 upperarm_twist_02_r {parent=upperarm_r, loc=<-13.7942 0 0>, rot=<0 0 0 1>}
 dyn_arm_measure_r {parent=clavicle_r, loc=<-36.5818 -0.0555 16.3194>, rot=<0.0008 0.0336 -0.0252 -0.9991>}
 dyn_parent_shoulderpad_r {parent=clavicle_r, loc=<-15.9168 0.7389 -0.4761>, rot=<0.0252 -0.9991 0.0008 -0.0336>}
 dyn_shoulderpad_base_r {parent=dyn_parent_shoulderpad_r, loc=<0 -0.769 0>, rot=<0.014 -0.9345 -0.0134 0.3555>}
 dyn_shoulderpad_r {parent=dyn_shoulderpad_base_r, loc=<0 0 0>, rot=<0 0 0 1>}
 dyn_shldr_feather_a_r {parent=dyn_shoulderpad_r, loc=<-3.7561 -9.04 -1.3152>, rot=<0.4722 -0.1834 -0.2349 -0.8296>}
 dyn_shldr_feather_b_r {parent=dyn_shoulderpad_r, loc=<-4.7799 -7.5912 -3.7374>, rot=<0.4508 -0.1585 -0.1663 -0.8626>}
 dyn_shldr_feather_c_r {parent=dyn_shoulderpad_r, loc=<-5.5701 -5.686 -6.0377>, rot=<0.3963 -0.1323 -0.0934 -0.9037>}
 dyn_shldr_feather_d_r {parent=dyn_shoulderpad_r, loc=<-4.8968 -3.1655 -8.394>, rot=<0.2322 -0.0921 -0.0347 -0.9677>}
 dyn_shldr_feather_e_r {parent=dyn_shoulderpad_r, loc=<-4.288 0.3181 -9.3422>, rot=<-0.0094 -0.0702 -0.0111 -0.9974>}
 dyn_shldr_feather_f_r {parent=dyn_shoulderpad_r, loc=<-4.4985 3.5053 -8.2773>, rot=<-0.2503 -0.0802 0.0426 -0.9639>}
 dyn_shldr_feather_g_r {parent=dyn_shoulderpad_r, loc=<-5.8639 6.3468 -5.9015>, rot=<-0.3815 -0.1392 0.065 -0.9115>}
 dyn_shldr_feather_h_r {parent=dyn_shoulderpad_r, loc=<-5.3896 8.1136 -3.3223>, rot=<-0.4441 -0.1712 0.1333 -0.8693>}
 dyn_shldr_feather_i_r {parent=dyn_shoulderpad_r, loc=<-4.4014 9.5222 -0.8812>, rot=<-0.4698 -0.2088 0.1914 -0.8361>}
 target_lowerarm_fwd {parent=spine_05, loc=<0.265 34.9718 -21.5372>, rot=<0 0 0.0544 0.9985>}
 thigh_l {parent=pelvis, loc=<-2.5693 -0.0109 -7.7501>, rot=<0.0012 0.0572 0.0155 -0.9982>}
 calf_l {parent=thigh_l, loc=<-39.0484 0 0>, rot=<0.0006 -0.0161 -0.0391 -0.9991>}
 calf_twist_01_l {parent=calf_l, loc=<-35.8131 -0.0967 0.0397>, rot=<0.0237 0.0095 -0.0035 -0.9997>}
 calf_twist_02_l {parent=calf_l, loc=<-18.8064 -0.0967 0.0397>, rot=<0.0237 0.0095 -0.0035 -0.9997>}
 foot_l {parent=calf_l, loc=<-42.561 0.1077 0.4292>, rot=<0.0418 -0.0455 0.0343 -0.9975>}
 ball_l {parent=foot_l, loc=<-6.1599 13.8264 0.6806>, rot=<0.0001 -0.0001 -0.7063 -0.7079>}
 thigh_twist_01_l {parent=thigh_l, loc=<-21.6303 0 0>, rot=<0.0013 0 0 -1>}
 dyn_volume_base_thigh_l {parent=thigh_l, loc=<-6 -9 0>, rot=<0 0 -0.4617 -0.887>}
 dyn_volume_thigh_l {parent=dyn_volume_base_thigh_l, loc=<0 0 0>, rot=<0 0 0 -1>}
 thigh_r {parent=pelvis, loc=<-2.5691 -0.0109 7.7501>, rot=<0.0572 -0.0012 0.9982 0.0155>}
 calf_r {parent=thigh_r, loc=<39.0487 0 0>, rot=<0.0006 -0.0161 -0.0391 -0.9991>}
 calf_twist_01_r {parent=calf_r, loc=<35.8132 0.0967 -0.0397>, rot=<0.0237 0.0095 -0.0035 -0.9997>}
 calf_twist_02_r {parent=calf_r, loc=<18.8064 0.0967 -0.0397>, rot=<0.0237 0.0095 -0.0035 -0.9997>}
 foot_r {parent=calf_r, loc=<42.5612 -0.1077 -0.4292>, rot=<0.0418 -0.0455 0.0343 -0.9975>}
 ball_r {parent=foot_r, loc=<6.1598 -13.8264 -0.6806>, rot=<0.0001 -0.0001 -0.7063 -0.7079>}
 thigh_twist_01_r {parent=thigh_r, loc=<21.6303 0 0>, rot=<0.0013 0 0 -1>}
 dyn_volume_base_thigh_r {parent=thigh_r, loc=<6 9 0>, rot=<0 0 -0.4617 -0.887>}
 dyn_volume_thigh_r {parent=dyn_volume_base_thigh_r, loc=<0 0 0>, rot=<0 0 0 1>}
 dyn_skirt_fr_a_base_l {parent=pelvis, loc=<4.8845 10.9402 -7.8463>, rot=<-0.0735 -0.9892 0.1197 -0.0426>}
 dyn_skirt_fr_a_l {parent=dyn_skirt_fr_a_base_l, loc=<0 0 0>, rot=<0 0 0 -1>}
 dyn_skirt_fr_b_l {parent=dyn_skirt_fr_a_l, loc=<15 0 0>, rot=<0 0 0 -1>}
 dyn_skirt_side_aa_base_l {parent=pelvis, loc=<5.648 3.3032 -15.3132>, rot=<-0.0874 -0.804 0.5863 -0.0455>}
 dyn_skirt_side_aa_l {parent=dyn_skirt_side_aa_base_l, loc=<0 0 0>, rot=<0 0 0 -1>}
 dyn_skirt_side_ab_l {parent=dyn_skirt_side_aa_l, loc=<15.974 0 0>, rot=<0 0 0 -1>}
 dyn_skirt_bk_a_base_l {parent=pelvis, loc=<4.3528 -15.2194 -5.5898>, rot=<-0.031 -0.074 0.9948 -0.0623>}
 dyn_skirt_bk_a_l {parent=dyn_skirt_bk_a_base_l, loc=<0 0 0>, rot=<0 0 0 -1>}
 dyn_skirt_bk_b_l {parent=dyn_skirt_bk_a_l, loc=<15.1073 0 0>, rot=<0 0 0 -1>}
 dyn_skirt_flap_aa_base_l {parent=pelvis, loc=<-9.7733 16.0293 -5.6704>, rot=<-0.0576 -0.9926 0.1072 -0.0024>}
 dyn_skirt_flap_aa_l {parent=dyn_skirt_flap_aa_base_l, loc=<0 0 0>, rot=<0 0 0 -1>}
 dyn_skirt_flap_ab_l {parent=dyn_skirt_flap_aa_l, loc=<9.8077 0 0>, rot=<0 0 0 -1>}
 dyn_skirt_flap_ba_base_l {parent=pelvis, loc=<-4.0251 12.0174 -11.9826>, rot=<-0.0694 -0.9408 0.3303 -0.0325>}
 dyn_skirt_flap_ba_l {parent=dyn_skirt_flap_ba_base_l, loc=<0 0 0>, rot=<0 0 0 1>}
 dyn_skirt_flap_bb_l {parent=dyn_skirt_flap_ba_l, loc=<12.7895 0 0>, rot=<0 0 0 1>}
 dyn_skirt_flap_ca_base_l {parent=pelvis, loc=<-0.6034 5.8078 -15.7304>, rot=<-0.0873 -0.8437 0.5256 -0.0657>}
 dyn_skirt_flap_ca_l {parent=dyn_skirt_flap_ca_base_l, loc=<0 0 0>, rot=<0 0 0 -1>}
 dyn_skirt_flap_cb_l {parent=dyn_skirt_flap_ca_l, loc=<16.6173 0 0>, rot=<0 0 0 -1>}
 dyn_skirt_flap_da_base_l {parent=pelvis, loc=<0.3702 -2.5881 -17.4549>, rot=<-0.0852 -0.7099 0.6957 -0.0686>}
 dyn_skirt_flap_da_l {parent=dyn_skirt_flap_da_base_l, loc=<0 0 0>, rot=<0 0 0 1>}
 dyn_skirt_flap_db_l {parent=dyn_skirt_flap_da_l, loc=<18.1414 0.0908 0>, rot=<0 0 0 1>}
 dyn_skirt_flap_ea_base_l {parent=pelvis, loc=<-3.4837 -10.1148 -16.4349>, rot=<-0.0666 -0.4932 0.8659 -0.0494>}
 dyn_skirt_flap_ea_l {parent=dyn_skirt_flap_ea_base_l, loc=<0 0 0>, rot=<0 0 0 -1>}
 dyn_skirt_flap_eb_l {parent=dyn_skirt_flap_ea_l, loc=<14.0607 0 0>, rot=<0 0 0 -1>}
 dyn_skirt_flap_fa_base_l {parent=pelvis, loc=<-10.4355 -16.2011 -11.4569>, rot=<-0.0497 -0.2535 0.9649 -0.0469>}
 dyn_skirt_flap_fa_l {parent=dyn_skirt_flap_fa_base_l, loc=<0 0 0>, rot=<0 0 0 1>}
 dyn_skirt_flap_fb_l {parent=dyn_skirt_flap_fa_l, loc=<7.169 0 0>, rot=<0 0 0 1>}
 dyn_skirt_flap_ga_base_l {parent=pelvis, loc=<-13.9335 -18.1571 -4.1345>, rot=<-0.0416 -0.0661 0.9959 -0.0467>}
 dyn_skirt_flap_ga_l {parent=dyn_skirt_flap_ga_base_l, loc=<0 0 0>, rot=<0 0 0 1>}
 dyn_skirt_flap_gb_l {parent=dyn_skirt_flap_ga_l, loc=<7 0 0>, rot=<0 0 0 1>}
 dyn_skirt_flap_aa_base_r {parent=pelvis, loc=<-9.7733 16.0293 5.6704>, rot=<0.9926 -0.0576 -0.0024 -0.1072>}
 dyn_skirt_flap_aa_r {parent=dyn_skirt_flap_aa_base_r, loc=<0 0 0>, rot=<0 0 0 1>}
 dyn_skirt_flap_ab_r {parent=dyn_skirt_flap_aa_r, loc=<-9.8077 0 0>, rot=<0 0 0 1>}
 dyn_skirt_flap_ba_base_r {parent=pelvis, loc=<-4.0251 12.0173 11.9826>, rot=<0.9408 -0.0694 -0.0325 -0.3303>}
 dyn_skirt_flap_ba_r {parent=dyn_skirt_flap_ba_base_r, loc=<0 0 0>, rot=<0 0 0 -1>}
 dyn_skirt_flap_bb_r {parent=dyn_skirt_flap_ba_r, loc=<-12.7895 0 0>, rot=<0 0 0 -1>}
 dyn_skirt_flap_ca_base_r {parent=pelvis, loc=<-0.6034 5.8078 15.7304>, rot=<0.8437 -0.0873 -0.0657 -0.5256>}
 dyn_skirt_flap_ca_r {parent=dyn_skirt_flap_ca_base_r, loc=<0 0 0>, rot=<0 0 0 -1>}
 dyn_skirt_flap_cb_r {parent=dyn_skirt_flap_ca_r, loc=<-16.6173 0 0>, rot=<0 0 0 -1>}
 dyn_skirt_flap_da_base_r {parent=pelvis, loc=<0.3702 -2.5881 17.4549>, rot=<0.7099 -0.0852 -0.0686 -0.6957>}
 dyn_skirt_flap_da_r {parent=dyn_skirt_flap_da_base_r, loc=<0 0 0>, rot=<0 0 0 -1>}
 dyn_skirt_flap_db_r {parent=dyn_skirt_flap_da_r, loc=<-18.1413 -0.0909 0>, rot=<0 0 0 -1>}
 dyn_skirt_flap_ea_base_r {parent=pelvis, loc=<-3.4837 -10.1148 16.4349>, rot=<0.4932 -0.0666 -0.0494 -0.8659>}
 dyn_skirt_flap_ea_r {parent=dyn_skirt_flap_ea_base_r, loc=<0 0 0>, rot=<0 0 0 -1>}
 dyn_skirt_flap_eb_r {parent=dyn_skirt_flap_ea_r, loc=<-14.0607 0 0>, rot=<0 0 0 -1>}
 dyn_skirt_flap_fa_base_r {parent=pelvis, loc=<-10.4355 -16.201 11.4569>, rot=<0.2535 -0.0497 -0.0469 -0.9649>}
 dyn_skirt_flap_fa_r {parent=dyn_skirt_flap_fa_base_r, loc=<0 0 0>, rot=<0 0 0 -1>}
 dyn_skirt_flap_fb_r {parent=dyn_skirt_flap_fa_r, loc=<-7.1691 -0.0001 0>, rot=<0 0 0 -1>}
 dyn_skirt_flap_ga_base_r {parent=pelvis, loc=<-13.9335 -18.1571 4.1345>, rot=<0.0661 -0.0416 -0.0467 -0.9959>}
 dyn_skirt_flap_ga_r {parent=dyn_skirt_flap_ga_base_r, loc=<0 0 0>, rot=<0 0 0 1>}
 dyn_skirt_flap_gb_r {parent=dyn_skirt_flap_ga_r, loc=<-7 0 0>, rot=<0 0 0 1>}
 dyn_flap_a_c_base {parent=pelvis, loc=<5.064 14.3235 0>, rot=<-0.1027 -0.9947 0 0>}
 dyn_flap_a_c {parent=dyn_flap_a_c_base, loc=<0 0 0>, rot=<0 0 0 -1>}
 dyn_flap_b_c {parent=dyn_flap_a_c, loc=<7.4418 0 0>, rot=<0 0 0 -1>}
 dyn_flap_c_c {parent=dyn_flap_b_c, loc=<7.4498 0 0>, rot=<0 0 0 -1>}
 dyn_flap_d_c {parent=dyn_flap_c_c, loc=<6.302 0 0>, rot=<0 0 0 -1>}
 dyn_flap_clip_c_base {parent=dyn_flap_a_c, loc=<3.3967 0 0>, rot=<0 0 0 -1>}
 dyn_flap_clip_c {parent=dyn_flap_clip_c_base, loc=<0 0 0>, rot=<0 0 0 -1>}
 dyn_skirt_side_ba_base_l {parent=pelvis, loc=<5.3698 -9.6644 -14.3255>, rot=<-0.0694 -0.5323 0.8414 -0.0619>}
 dyn_skirt_side_ba_l {parent=dyn_skirt_side_ba_base_l, loc=<0 0 0>, rot=<0 0 0 -1>}
 dyn_skirt_side_bb_l {parent=dyn_skirt_side_ba_l, loc=<15.974 0 0>, rot=<0 0 0 -1>}
 dyn_skirt_fr_a_base_r {parent=pelvis, loc=<4.8845 10.9402 7.8462>, rot=<0.9892 -0.0735 -0.0426 -0.1197>}
 dyn_skirt_fr_a_r {parent=dyn_skirt_fr_a_base_r, loc=<0 0 0>, rot=<0 0 0 1>}
 dyn_skirt_fr_b_r {parent=dyn_skirt_fr_a_r, loc=<-15 0 0>, rot=<0 0 0 1>}
 dyn_skirt_side_aa_base_r {parent=pelvis, loc=<5.648 3.3032 15.3132>, rot=<0.8045 -0.0828 -0.0418 -0.5867>}
 dyn_skirt_side_aa_r {parent=dyn_skirt_side_aa_base_r, loc=<0 0 0>, rot=<0 0 0 -1>}
 dyn_skirt_side_ab_r {parent=dyn_skirt_side_aa_r, loc=<-15.974 0 0>, rot=<0 0 0 -1>}
 dyn_skirt_bk_a_base_r {parent=pelvis, loc=<4.3529 -15.2194 5.5898>, rot=<0.074 -0.031 -0.0623 -0.9948>}
 dyn_skirt_bk_a_r {parent=dyn_skirt_bk_a_base_r, loc=<0 0 0>, rot=<0 0 0 -1>}
 dyn_skirt_bk_b_r {parent=dyn_skirt_bk_a_r, loc=<-15.1073 0 0>, rot=<0 0 0 -1>}
 dyn_skirt_side_ba_base_r {parent=pelvis, loc=<5.3698 -9.6644 14.3255>, rot=<0.5334 -0.0581 -0.0584 -0.8418>}
 dyn_skirt_side_ba_r {parent=dyn_skirt_side_ba_base_r, loc=<0 0 0>, rot=<0 0 0 1>}
 dyn_skirt_side_bb_r {parent=dyn_skirt_side_ba_r, loc=<-15.974 0 0.0001>, rot=<0 0 0 1>}
]
