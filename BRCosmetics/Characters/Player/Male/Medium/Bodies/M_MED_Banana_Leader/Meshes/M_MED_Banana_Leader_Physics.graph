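# Physics asset for the M_MED_Banana_Leader player body, covering only the legs and the dyn_skirt bones.
# Eight skirt chains (front, back and two sides, left and right) each link a tiny base capsule to two larger, offset capsules.
# Skirt joints allow more swing on one axis than in the other swing or twist.
# Thigh and calf capsules give the legs collision shapes, with fully free angular knee constraints.
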PhysicsAsset /BRCosmetics/Characters/Player/Male/Medium/Bodies/M_MED_Banana_Leader/Meshes/M_MED_Banana_Leader_Physics
bodies {
 calf_l [{shapeType=Capsule, length=45, radius=7.25, center=<-23.4706 0.1092 0.0001>, rotation=<rot 90 0 5.1094>}]
 calf_r [{shapeType=Capsule, length=45, radius=7.25, center=<23.4706 -0.1092 0>, rotation=<rot 90 0 -174.8907>}]
 dyn_skirt_bk_a_base_l [{shapeType=Capsule, length=0.505, radius=0.505, rotation=<rot 0 -90 0>}]
 dyn_skirt_bk_a_base_r [{shapeType=Capsule, length=0.505, radius=0.505, rotation=<rot 0 -90 180>}]
 dyn_skirt_bk_a_l [{shapeType=Capsule, length=5.5, radius=10, center=<10 5 0>, rotation=<rot 90 0 0>}]
 dyn_skirt_bk_a_r [{shapeType=Capsule, length=5.5, radius=10, center=<-10 -5 0>, rotation=<rot 90 0 180>}]
 dyn_skirt_bk_b_l [{shapeType=Capsule, length=4, radius=10.25, center=<11 5 0>, rotation=<rot 90 0 0>}]
 dyn_skirt_bk_b_r [{shapeType=Capsule, length=4, radius=10.25, center=<-11 -5 0>, rotation=<rot 90 0 180>}]
 dyn_skirt_fr_a_base_l [{shapeType=Capsule, length=0.505, radius=0.505, rotation=<rot 0 -90 0>}]
 dyn_skirt_fr_a_base_r [{shapeType=Capsule, length=0.505, radius=0.505, rotation=<rot 0 -90 180>}]
 dyn_skirt_fr_a_l [{shapeType=Capsule, length=5.5, radius=10, center=<10 7 0>, rotation=<rot 90 0 0>}]
 dyn_skirt_fr_a_r [{shapeType=Capsule, length=5.5, radius=10, center=<-10 -7 0>, rotation=<rot 90 0 180>}]
 dyn_skirt_fr_b_l [{shapeType=Capsule, length=4, radius=10.25, center=<12 7 0>, rotation=<rot 90 0 0>}]
 dyn_skirt_fr_b_r [{shapeType=Capsule, length=4, radius=10.25, center=<-12 -7 0>, rotation=<rot 90 0 180>}]
 dyn_skirt_side_aa_base_l [{shapeType=Capsule, length=0.505, radius=0.505, rotation=<rot 0 -90 0>}]
 dyn_skirt_side_aa_base_r [{shapeType=Capsule, length=0.505, radius=0.505, rotation=<rot 0 -90 180>}]
 dyn_skirt_side_aa_l [{shapeType=Capsule, length=5.5, radius=10, center=<12 7 0>, rotation=<rot 90 0 0>}]
 dyn_skirt_side_aa_r [{shapeType=Capsule, length=5.5, radius=10, center=<-12 -7 0>, rotation=<rot 90 0 180>}]
 dyn_skirt_side_ab_l [{shapeType=Capsule, length=4, radius=10.25, center=<12 7 0>, rotation=<rot 90 0 0>}]
 dyn_skirt_side_ab_r [{shapeType=Capsule, length=4, radius=10.25, center=<-12 -7 0>, rotation=<rot 90 0 180>}]
 dyn_skirt_side_ba_base_l [{shapeType=Capsule, length=0.505, radius=0.505, rotation=<rot 0 -90 0>}]
 dyn_skirt_side_ba_base_r [{shapeType=Capsule, length=0.505, radius=0.505, rotation=<rot 0 -90 180>}]
 dyn_skirt_side_ba_l [{shapeType=Capsule, length=5.5, radius=10, center=<12 7 0>, rotation=<rot 90 0 0>}]
 dyn_skirt_side_ba_r [{shapeType=Capsule, length=5.5, radius=10, center=<-12 -7 0>, rotation=<rot 90 0 180>}]
 dyn_skirt_side_bb_l [{shapeType=Capsule, length=4, radius=10.25, center=<12 7 0>, rotation=<rot 90 0 0>}]
 dyn_skirt_side_bb_r [{shapeType=Capsule, length=4, radius=10.25, center=<-12 -7 0>, rotation=<rot 90 0 180>}]
 thigh_l [{shapeType=Capsule, length=24.25, radius=8.25, center=<-20.2086 0 0.0001>, rotation=<rot 90 0 0>}]
 thigh_r [{shapeType=Capsule, length=24.25, radius=8.25, center=<20.2086 0 -0.0001>, rotation=<rot 90 0 180>}]
}
constraints [
 {
  bone1Name=calf_l
  bone2Name=thigh_l
  swing1LimitDegrees=45
  swing1Motion=Free
  swing2LimitDegrees=45
  swing2Motion=Free
  twistLimitDegrees=45
  twistMotion=Free
 }
 {
  bone1Name=calf_r
  bone2Name=thigh_r
  swing1LimitDegrees=45
  swing1Motion=Free
  swing2LimitDegrees=45
  swing2Motion=Free
  twistLimitDegrees=45
  twistMotion=Free
 }
 {bone1Name=dyn_skirt_fr_a_l, bone2Name=dyn_skirt_fr_a_base_l, swing1LimitDegrees=25, swing2LimitDegrees=5, twistLimitDegrees=5}
 {bone1Name=dyn_skirt_fr_b_l, bone2Name=dyn_skirt_fr_a_l, swing1LimitDegrees=25, swing2LimitDegrees=5, twistLimitDegrees=5}
 {bone1Name=dyn_skirt_bk_a_l, bone2Name=dyn_skirt_bk_a_base_l, swing1LimitDegrees=25, swing2LimitDegrees=5, twistLimitDegrees=5}
 {bone1Name=dyn_skirt_bk_b_l, bone2Name=dyn_skirt_bk_a_l, swing1LimitDegrees=25, swing2LimitDegrees=5, twistLimitDegrees=5}
 {bone1Name=dyn_skirt_fr_a_r, bone2Name=dyn_skirt_fr_a_base_r, swing1LimitDegrees=25, swing2LimitDegrees=5, twistLimitDegrees=5}
 {bone1Name=dyn_skirt_fr_b_r, bone2Name=dyn_skirt_fr_a_r, swing1LimitDegrees=25, swing2LimitDegrees=5, twistLimitDegrees=5}
 {bone1Name=dyn_skirt_bk_a_r, bone2Name=dyn_skirt_bk_a_base_r, swing1LimitDegrees=25, swing2LimitDegrees=5, twistLimitDegrees=5}
 {bone1Name=dyn_skirt_bk_b_r, bone2Name=dyn_skirt_bk_a_r, swing1LimitDegrees=25, swing2LimitDegrees=5, twistLimitDegrees=5}
 {bone1Name=dyn_skirt_side_aa_l, bone2Name=dyn_skirt_side_aa_base_l, swing1LimitDegrees=25, swing2LimitDegrees=5, twistLimitDegrees=5}
 {bone1Name=dyn_skirt_side_ab_l, bone2Name=dyn_skirt_side_aa_l, swing1LimitDegrees=25, swing2LimitDegrees=5, twistLimitDegrees=5}
 {bone1Name=dyn_skirt_side_ba_l, bone2Name=dyn_skirt_side_ba_base_l, swing1LimitDegrees=25, swing2LimitDegrees=5, twistLimitDegrees=5}
 {bone1Name=dyn_skirt_side_bb_l, bone2Name=dyn_skirt_side_ba_l, swing1LimitDegrees=25, swing2LimitDegrees=5, twistLimitDegrees=5}
 {bone1Name=dyn_skirt_side_aa_r, bone2Name=dyn_skirt_side_aa_base_r, swing1LimitDegrees=25, swing2LimitDegrees=5, twistLimitDegrees=5}
 {bone1Name=dyn_skirt_side_ab_r, bone2Name=dyn_skirt_side_aa_r, swing1LimitDegrees=25, swing2LimitDegrees=5, twistLimitDegrees=5}
 {bone1Name=dyn_skirt_side_ba_r, bone2Name=dyn_skirt_side_ba_base_r, swing1LimitDegrees=25, swing2LimitDegrees=5, twistLimitDegrees=5}
 {bone1Name=dyn_skirt_side_bb_r, bone2Name=dyn_skirt_side_ba_r, swing1LimitDegrees=25, swing2LimitDegrees=5, twistLimitDegrees=5}
]
