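# Post-pose rig AnimBlueprint for the Banana Leader body cosmetic, adding secondary motion to shoulder pads, skirt and flaps.
# Shoulder pads shift during parachute and skydive float; skirt bases follow thighs and pelvis, with thigh-driven offsets, then RigidBody simulation.
# Weighted constraints derive skirt flap bones from simulated skirt bones, avoiding per-flap simulation.
# The center flap switches setups on crouch, adds crouch-scaled AnimDynamics, then positions its clip.
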AnimBlueprint /BRCosmetics/Characters/Player/Male/Medium/Bodies/M_MED_Banana_Leader/Meshes/M_MED_Banana_Leader_AnimBP
target_skeleton=&/M_MED_Banana_Leader_Skeleton
graphs [
 AnimGraph {nodes [
   n1 = Get bIsParachuteOpened()
   n2 = "Transform (Modify) Bone - Bone: dyn_shoulderpad_base_l"!(ComponentPose, bAlphaBoolEnabled:bool=@n1, Translation="7,-3,0", Rotation="-5,0,-50", Scale="1,1,1") -> (Pose) {
    alphaInputType=Bool
    rotationMode=BMM_Additive
    rotationSpace=BCS_BoneSpace
    scaleMode=BMM_Additive
    scaleSpace=BCS_BoneSpace
    translationMode=BMM_Additive
    translationSpace=BCS_BoneSpace
    alphaBoolBlend {blendOption=Linear}
    alphaScaleBiasClamp {inRange {max=25}}
    boneToModify {boneName=dyn_shoulderpad_base_l}
    rotation {pitch=-5, roll=-50}
    translation=<7 -3>
   }
   n3 = "Transform (Modify) Bone - Bone: dyn_shoulderpad_base_r"!(ComponentPose=@n2, bAlphaBoolEnabled:bool=@n1, Translation="-8,3,0", Rotation="-5,0,-50", Scale="1,1,1") -> (Pose) {
    alphaInputType=Bool
    rotationMode=BMM_Additive
    rotationSpace=BCS_BoneSpace
    scaleMode=BMM_Additive
    scaleSpace=BCS_BoneSpace
    translationMode=BMM_Additive
    translationSpace=BCS_BoneSpace
    alphaBoolBlend {blendOption=Linear}
    alphaScaleBiasClamp {inRange {max=25}}
    boneToModify {boneName=dyn_shoulderpad_base_r}
    rotation {pitch=-5, roll=-50}
    translation=<-8 3>
   }
   n4 = Get bIsSkydiveFloating()
   n5 = "Transform (Modify) Bone - Bone: dyn_shoulderpad_base_l"(ComponentPose=@n3, bAlphaBoolEnabled=@n4, Translation="8,0,0", Rotation="0,0,0", Scale="1,1,1") {
    alphaInputType=Bool
    rotationMode=BMM_Additive
    rotationSpace=BCS_BoneSpace
    scaleMode=BMM_Additive
    scaleSpace=BCS_BoneSpace
    translationMode=BMM_Additive
    translationSpace=BCS_BoneSpace
    alphaBoolBlend {blendOption=Linear}
    alphaScaleBiasClamp {inRange {max=25}}
    boneToModify {boneName=dyn_shoulderpad_base_l}
    translation {x=8}
   }
   "Transform (Modify) Bone - Bone: dyn_shoulderpad_base_r"(ComponentPose=@n5, bAlphaBoolEnabled=@n4, Translation="-8,0,3", Rotation="0,0,0", Scale="1,1,1") {
    alphaInputType=Bool
    rotationMode=BMM_Additive
    rotationSpace=BCS_BoneSpace
    scaleMode=BMM_Additive
    scaleSpace=BCS_BoneSpace
    translationMode=BMM_Additive
    translationSpace=BCS_BoneSpace
    alphaBoolBlend {blendOption=Linear}
    alphaScaleBiasClamp {inRange {max=25}}
    boneToModify {boneName=dyn_shoulderpad_base_r}
    translation {x=-8, z=3}
   }
   n6 = "Constraint - dyn_flap_a_c_base (RR)"!(
    ComponentPose=Local To Component(LocalPose=Use cached pose 'SkirtFlaps_Pos'())
    Alpha=1
    ConstraintWeights_0=0.5
    ConstraintWeights_1:float=0.5
   ) -> (Pose) {
    alphaBoolBlend {blendOption=Linear}
    boneToModify {boneName=dyn_flap_a_c_base}
    constraintSetup [{transformType=Rotation, targetBone {boneName=thigh_l}}, {transformType=Rotation, targetBone {boneName=thigh_r}}]
    constraintWeights [1,1]
   }
   n7 = "Transform (Modify) Bone - Bone: dyn_flap_a_c_base"(ComponentPose=@n6, Alpha=Get thigh_rot_l(), Translation="0,3,0", Rotation="5,5,50", Scale="1,1,1") {
    rotationMode=BMM_Additive
    rotationSpace=BCS_BoneSpace
    scaleMode=BMM_Additive
    scaleSpace=BCS_BoneSpace
    translationMode=BMM_Additive
    translationSpace=BCS_BoneSpace
    alphaBoolBlend {blendOption=Linear}
    alphaScaleBiasClamp {bMapRange=true, inRange {max=0, min=-100}, outRange {max=0, min=1}}
    boneToModify {boneName=dyn_flap_a_c_base}
    rotation=<rot 5 5 50>
    translation {y=3}
   }
   n8 = "Transform (Modify) Bone - Bone: dyn_flap_a_c_base"(ComponentPose=@n7, Alpha=Get thigh_rot_r(), Translation="0,3,0", Rotation="10,5,-50", Scale="1,1,1") {
    rotationMode=BMM_Additive
    rotationSpace=BCS_BoneSpace
    scaleMode=BMM_Additive
    scaleSpace=BCS_BoneSpace
    translationMode=BMM_Additive
    translationSpace=BCS_BoneSpace
    alphaBoolBlend {blendOption=Linear}
    alphaScaleBiasClamp {bMapRange=true, inRange {max=0, min=-100}, outRange {max=0, min=1}}
    boneToModify {boneName=dyn_flap_a_c_base}
    rotation=<rot 10 5 -50>
    translation {y=3}
   }
   n9 = Component To Local(ComponentPose=@n8)
   n10 = "Constraint - dyn_flap_a_c_base (PPPP)"!(
    ComponentPose=Local To Component(LocalPose=Use cached pose 'SkirtFlaps_Pos'())
    Alpha=1
    ConstraintWeights_0=0.15
    ConstraintWeights_1:float=0.3
    ConstraintWeights_2:float=0.3
    ConstraintWeights_3:float=0.25
   ) -> (Pose) {
    alphaBoolBlend {blendOption=Linear}
    boneToModify {boneName=dyn_flap_a_c_base}
    constraintSetup [
     {targetBone {boneName=spine_01}}
     {targetBone {boneName=dyn_skirt_fr_a_r}}
     {targetBone {boneName=dyn_skirt_fr_a_l}}
     {targetBone {boneName=pelvis}}
    ]
    constraintWeights [1,1,1,1]
   }
   n11 = "Constraint - dyn_flap_b_c (PPPPP)"!(
    ComponentPose=@n10
    Alpha=1
    ConstraintWeights_0=0.1
    ConstraintWeights_1:float=0.1
    ConstraintWeights_2:float=0.35
    ConstraintWeights_3:float=0.1
    ConstraintWeights_4:float=0.35
   ) -> (Pose) {
    alphaBoolBlend {blendOption=Linear}
    boneToModify {boneName=dyn_flap_b_c}
    constraintSetup [
     {targetBone {boneName=pelvis}}
     {targetBone {boneName=dyn_skirt_fr_b_l}}
     {targetBone {boneName=dyn_skirt_fr_a_r}}
     {targetBone {boneName=dyn_skirt_fr_b_r}}
     {targetBone {boneName=dyn_skirt_fr_a_l}}
    ]
    constraintWeights [1,1,1,1,1]
   }
   n12 = "Constraint - dyn_flap_c_c (PPPP)"!(
    ComponentPose=@n11
    Alpha=1
    ConstraintWeights_0=0.35
    ConstraintWeights_1:float=0.15
    ConstraintWeights_2:float=0.35
    ConstraintWeights_3:float=0.15
   ) -> (Pose) {
    alphaBoolBlend {blendOption=Linear}
    boneToModify {boneName=dyn_flap_c_c}
    constraintSetup [
     {targetBone {boneName=dyn_skirt_fr_b_l}}
     {targetBone {boneName=dyn_skirt_fr_a_r}}
     {targetBone {boneName=dyn_skirt_fr_b_r}}
     {targetBone {boneName=dyn_skirt_fr_a_l}}
    ]
    constraintWeights [1,1,1,1]
   }
   n13 = "Constraint - dyn_flap_d_c (PP)"!(ComponentPose=@n12, Alpha=1, ConstraintWeights_0=0.5, ConstraintWeights_1:float=0.5) -> (Pose) {
    alphaBoolBlend {blendOption=Linear}
    boneToModify {boneName=dyn_flap_d_c}
    constraintSetup [{targetBone {boneName=dyn_skirt_fr_b_l}}, {targetBone {boneName=dyn_skirt_fr_b_r}}]
    constraintWeights [1,1]
   }
   n14 = Get thigh_rot_l()
   n15 = "Transform (Modify) Bone - Bone: dyn_flap_a_c_base"(ComponentPose=@n13, Alpha=@n14, Translation="0,1.5,0", Rotation="5,0,20", Scale="1,1,1") {
    rotationMode=BMM_Additive
    rotationSpace=BCS_BoneSpace
    scaleMode=BMM_Additive
    scaleSpace=BCS_BoneSpace
    translationMode=BMM_Additive
    translationSpace=BCS_BoneSpace
    alphaBoolBlend {blendOption=Linear}
    alphaScaleBiasClamp {bMapRange=true, inRange {max=0, min=-100}, outRange {max=0, min=1}}
    boneToModify {boneName=dyn_flap_a_c_base}
    rotation {pitch=5, roll=20}
    translation {y=1.5}
   }
   n16 = "Transform (Modify) Bone - Bone: dyn_flap_b_c"(ComponentPose=@n15, Alpha=@n14, Translation="0,1.5,0", Rotation="15,0,30", Scale="1,1,1") {
    rotationMode=BMM_Additive
    rotationSpace=BCS_BoneSpace
    scaleMode=BMM_Additive
    scaleSpace=BCS_BoneSpace
    translationMode=BMM_Additive
    translationSpace=BCS_BoneSpace
    alphaBoolBlend {blendOption=Linear}
    alphaScaleBiasClamp {bMapRange=true, inRange {max=0, min=-100}, outRange {max=0, min=1}}
    boneToModify {boneName=dyn_flap_b_c}
    rotation {pitch=15, roll=30}
    translation {y=1.5}
   }
   n17 = Get thigh_rot_r()
   n18 = "Transform (Modify) Bone - Bone: dyn_flap_a_c_base"(ComponentPose=@n16, Alpha=@n17, Translation="0,1.5,0", Rotation="-5,0,-20", Scale="1,1,1") {
    rotationMode=BMM_Additive
    rotationSpace=BCS_BoneSpace
    scaleMode=BMM_Additive
    scaleSpace=BCS_BoneSpace
    translationMode=BMM_Additive
    translationSpace=BCS_BoneSpace
    alphaBoolBlend {blendOption=Linear}
    alphaScaleBiasClamp {bMapRange=true, inRange {max=0, min=-100}, outRange {max=0, min=1}}
    boneToModify {boneName=dyn_flap_a_c_base}
    rotation {pitch=-5, roll=-20}
    translation {y=1.5}
   }
   n19 = "Transform (Modify) Bone - Bone: dyn_flap_b_c"(ComponentPose=@n18, Alpha=@n17, Translation="0,1.5,0", Rotation="-15,0,-30", Scale="1,1,1") {
    rotationMode=BMM_Additive
    rotationSpace=BCS_BoneSpace
    scaleMode=BMM_Additive
    scaleSpace=BCS_BoneSpace
    translationMode=BMM_Additive
    translationSpace=BCS_BoneSpace
    alphaBoolBlend {blendOption=Linear}
    alphaScaleBiasClamp {bMapRange=true, inRange {max=0, min=-100}, outRange {max=0, min=1}}
    boneToModify {boneName=dyn_flap_b_c}
    rotation {pitch=-15, roll=-30}
    translation {y=1.5}
   }
   n20 = Component To Local(ComponentPose=@n19)
   n21 = Get bIsCrouching()
   n22 = Get isCrouchingAlpha() -> (isCrouchingAlpha:double)
   n23 = Anim Dynamics - Bone: dyn_flap_b_c(
    ComponentPose=Local To Component(LocalPose=Blend Poses by bool(BlendPose_0=@n9, BlendPose_1=@n20, bActiveValue=@n21) {blendType=Linear})
    Alpha=@n22
    ExternalForce="0,0,0"
   ) {
    angularDampingOverride=1
    angularSpringConstant=1500
    bAngularSpring=true
    bGravityOverrideInSimSpace=true
    bOverrideAngularDamping=true
    gravityScale=0.35
    simulationSpace=RootRelative
    alphaBoolBlend {blendOption=Linear}
    alphaScaleBiasClamp {bMapRange=true, outRange {max=0.2, min=1}}
    boundBone {boneName=dyn_flap_b_c}
    physicsBodyDefinitions [{
      sphereCollisionRadius=0
      boundBone {boneName=dyn_flap_b_c}
      boxExtents=<10 20 20>
      constraintSetup {angularLimitsMax=<0 0 20>, angularLimitsMin=<0 0 0>, angularTarget=<0 0 0>}
      localJointOffset=<10 0 0>
     }]
   }
   n24 = Anim Dynamics - Bone: dyn_flap_c_c(ComponentPose=@n23, Alpha=@n22, ExternalForce="0,0,0") {
    angularDampingOverride=1
    angularSpringConstant=1500
    bAngularSpring=true
    bGravityOverrideInSimSpace=true
    bOverrideAngularDamping=true
    gravityScale=0.35
    simulationSpace=RootRelative
    alphaBoolBlend {blendOption=Linear}
    alphaScaleBiasClamp {bMapRange=true, outRange {max=0.2, min=1}}
    boundBone {boneName=dyn_flap_c_c}
    physicsBodyDefinitions [{
      sphereCollisionRadius=0
      boundBone {boneName=dyn_flap_c_c}
      boxExtents=<10 20 20>
      constraintSetup {angularLimitsMax=<0 0 20>, angularLimitsMin=<0 0 0>, angularTarget=<0 0 0>}
      localJointOffset=<10 0 0>
     }]
   }
   n25 = Anim Dynamics - Bone: dyn_flap_d_c(ComponentPose=@n24, Alpha=@n22, ExternalForce="0,0,0") {
    angularDampingOverride=1
    angularSpringConstant=1500
    bAngularSpring=true
    bGravityOverrideInSimSpace=true
    bOverrideAngularDamping=true
    gravityScale=0.35
    simulationSpace=RootRelative
    alphaBoolBlend {blendOption=Linear}
    alphaScaleBiasClamp {bMapRange=true, outRange {max=0.2, min=1}}
    boundBone {boneName=dyn_flap_d_c}
    physicsBodyDefinitions [{
      sphereCollisionRadius=0
      boundBone {boneName=dyn_flap_d_c}
      boxExtents=<10 20 20>
      constraintSetup {angularLimitsMax=<0 0 20>, angularLimitsMin=<0 0 0>, angularTarget=<0 0 0>}
      localJointOffset=<10 0 0>
     }]
   }
   n26 = "Constraint - dyn_flap_clip_c_base (PPPPPPP)"!(
    ComponentPose=@n25
    Alpha=1
    ConstraintWeights_0=0.3605
    ConstraintWeights_1:float=0.3128
    ConstraintWeights_2:float=0.0624
    ConstraintWeights_3:float=0.0605
    ConstraintWeights_4:float=0.06
    ConstraintWeights_5:float=0.0574
    ConstraintWeights_6:float=0.0864
   ) -> (Pose) {
    alphaBoolBlend {blendOption=Linear}
    boneToModify {boneName=dyn_flap_clip_c_base}
    constraintSetup [
     {targetBone {boneName=dyn_flap_a_c}}
     {targetBone {boneName=dyn_flap_b_c}}
     {targetBone {boneName=dyn_skirt_fr_a_l}}
     {targetBone {boneName=spine_01}}
     {targetBone {boneName=dyn_skirt_fr_a_r}}
     {targetBone {boneName=dyn_flap_c_c}}
     {targetBone {boneName=pelvis}}
    ]
    constraintWeights [1,1,1,1,1,1,1]
   }
   n27 = "Transform (Modify) Bone - Bone: dyn_flap_clip_c_base"(ComponentPose=@n26, Alpha=1, Translation="1.2,0.5,0", Rotation="0,7,0", Scale="0.8,1,0.8") {
    rotationMode=BMM_Additive
    rotationSpace=BCS_BoneSpace
    scaleMode=BMM_Additive
    scaleSpace=BCS_BoneSpace
    translationMode=BMM_Additive
    translationSpace=BCS_BoneSpace
    alphaBoolBlend {blendOption=Linear}
    alphaScaleBiasClamp {inRange {max=0, min=-100}, outRange {max=0, min=1}}
    boneToModify {boneName=dyn_flap_clip_c_base}
    rotation {yaw=7}
    scale {x=0.8, z=0.8}
    translation=<1.2 0.5>
   }
   n28 = "Transform (Modify) Bone - Bone: dyn_flap_clip_c_base"(ComponentPose=@n27, Alpha=Get thigh_rot_l(), Translation="1,0,0", Rotation="0,10,0", Scale="1,1,1") {
    rotationMode=BMM_Additive
    rotationSpace=BCS_BoneSpace
    scaleMode=BMM_Additive
    scaleSpace=BCS_BoneSpace
    translationMode=BMM_Additive
    translationSpace=BCS_BoneSpace
    alphaBoolBlend {blendOption=Linear}
    alphaScaleBiasClamp {bMapRange=true, inRange {max=0, min=-100}, outRange {max=0, min=1}}
    boneToModify {boneName=dyn_flap_clip_c_base}
    rotation {yaw=10}
    translation {x=1}
   }
   n29 = "Transform (Modify) Bone - Bone: dyn_flap_clip_c_base"(ComponentPose=@n28, Alpha=Get thigh_rot_r(), Translation="1,0,0", Rotation="0,10,0", Scale="1,1,1") {
    rotationMode=BMM_Additive
    rotationSpace=BCS_BoneSpace
    scaleMode=BMM_Additive
    scaleSpace=BCS_BoneSpace
    translationMode=BMM_Additive
    translationSpace=BCS_BoneSpace
    alphaBoolBlend {blendOption=Linear}
    alphaScaleBiasClamp {bMapRange=true, inRange {max=0, min=-100}, outRange {max=0, min=1}}
    boneToModify {boneName=dyn_flap_clip_c_base}
    rotation {yaw=10}
    translation {x=1}
   }
   Output Pose(Result=Component To Local(ComponentPose=@n29)) {layerGroup=None}
   n30 = "Constraint - dyn_skirt_fr_a_base_l (RR)"!(
    ComponentPose=Local To Component(LocalPose=Use cached pose 'ShoulderPads_Pose'())
    Alpha=1
    ConstraintWeights_0=0.9
    ConstraintWeights_1:float=0.1
   ) -> (Pose) {
    alphaBoolBlend {blendOption=Linear}
    boneToModify {boneName=dyn_skirt_fr_a_base_l}
    constraintSetup [{transformType=Rotation, targetBone {boneName=thigh_l}}, {transformType=Rotation, targetBone {boneName=pelvis}}]
    constraintWeights [1,1]
   }
   n31 = "Constraint - dyn_skirt_side_aa_base_l (RR)"!(ComponentPose=@n30, Alpha=1, ConstraintWeights_0=0.75, ConstraintWeights_1:float=0.25) -> (Pose) {
    alphaBoolBlend {blendOption=Linear}
    boneToModify {boneName=dyn_skirt_side_aa_base_l}
    constraintSetup [{transformType=Rotation, targetBone {boneName=thigh_l}}, {transformType=Rotation, targetBone {boneName=pelvis}}]
    constraintWeights [1,1]
   }
   n32 = "Constraint - dyn_skirt_side_ba_base_l (RR)"!(ComponentPose=@n31, Alpha=1, ConstraintWeights_0=0.5, ConstraintWeights_1:float=0.5) -> (Pose) {
    alphaBoolBlend {blendOption=Linear}
    boneToModify {boneName=dyn_skirt_side_ba_base_l}
    constraintSetup [{transformType=Rotation, targetBone {boneName=thigh_l}}, {transformType=Rotation, targetBone {boneName=pelvis}}]
    constraintWeights [1,1]
   }
   n33 = "Constraint - dyn_skirt_bk_a_base_l (RR)"!(ComponentPose=@n32, Alpha=1, ConstraintWeights_0=0.25, ConstraintWeights_1:float=0.75) -> (Pose) {
    alphaBoolBlend {blendOption=Linear}
    boneToModify {boneName=dyn_skirt_bk_a_base_l}
    constraintSetup [{transformType=Rotation, targetBone {boneName=thigh_l}}, {transformType=Rotation, targetBone {boneName=pelvis}}]
    constraintWeights [1,1]
   }
   n34 = "Constraint - dyn_skirt_fr_a_base_r (RR)"!(ComponentPose=@n33, Alpha=1, ConstraintWeights_0=0.9, ConstraintWeights_1:float=0.1) -> (Pose) {
    alphaBoolBlend {blendOption=Linear}
    boneToModify {boneName=dyn_skirt_fr_a_base_r}
    constraintSetup [{transformType=Rotation, targetBone {boneName=thigh_r}}, {transformType=Rotation, targetBone {boneName=pelvis}}]
    constraintWeights [1,1]
   }
   n35 = "Constraint - dyn_skirt_side_aa_base_r (RR)"!(ComponentPose=@n34, Alpha=1, ConstraintWeights_0=0.75, ConstraintWeights_1:float=0.25) -> (Pose) {
    alphaBoolBlend {blendOption=Linear}
    boneToModify {boneName=dyn_skirt_side_aa_base_r}
    constraintSetup [{transformType=Rotation, targetBone {boneName=thigh_r}}, {transformType=Rotation, targetBone {boneName=pelvis}}]
    constraintWeights [1,1]
   }
   n36 = "Constraint - dyn_skirt_side_ba_base_r (RR)"!(ComponentPose=@n35, Alpha=1, ConstraintWeights_0=0.5, ConstraintWeights_1:float=0.5) -> (Pose) {
    alphaBoolBlend {blendOption=Linear}
    boneToModify {boneName=dyn_skirt_side_ba_base_r}
    constraintSetup [{transformType=Rotation, targetBone {boneName=thigh_r}}, {transformType=Rotation, targetBone {boneName=pelvis}}]
    constraintWeights [1,1]
   }
   n37 = "Constraint - dyn_skirt_bk_a_base_r (RR)"!(ComponentPose=@n36, Alpha=1, ConstraintWeights_0=0.25, ConstraintWeights_1:float=0.75) -> (Pose) {
    alphaBoolBlend {blendOption=Linear}
    boneToModify {boneName=dyn_skirt_bk_a_base_r}
    constraintSetup [{transformType=Rotation, targetBone {boneName=thigh_r}}, {transformType=Rotation, targetBone {boneName=pelvis}}]
    constraintWeights [1,1]
   }
   n38 = Get thigh_rot_l()
   n39 = "Transform (Modify) Bone - Bone: dyn_skirt_fr_a_base_l"(ComponentPose=@n37, Alpha=@n38, Translation="2,5,0", Rotation="10,5,0", Scale="1,1,1") {
    rotationMode=BMM_Additive
    rotationSpace=BCS_BoneSpace
    scaleMode=BMM_Additive
    scaleSpace=BCS_BoneSpace
    translationMode=BMM_Additive
    translationSpace=BCS_BoneSpace
    alphaBoolBlend {blendOption=Linear}
    alphaScaleBiasClamp {bMapRange=true, inRange {max=0, min=-100}, outRange {max=0, min=1}}
    boneToModify {boneName=dyn_skirt_fr_a_base_l}
    rotation {pitch=10, yaw=5}
    translation=<2 5>
   }
   n40 = "Transform (Modify) Bone - Bone: dyn_skirt_side_aa_base_l"(ComponentPose=@n39, Alpha=@n38, Translation="0,5,-2", Rotation="-10,5,0", Scale="1,1,1") {
    rotationMode=BMM_Additive
    rotationSpace=BCS_BoneSpace
    scaleMode=BMM_Additive
    scaleSpace=BCS_BoneSpace
    translationMode=BMM_Additive
    translationSpace=BCS_BoneSpace
    alphaBoolBlend {blendOption=Linear}
    alphaScaleBiasClamp {bMapRange=true, inRange {max=0, min=-100}, outRange {max=0, min=1}}
    boneToModify {boneName=dyn_skirt_side_aa_base_l}
    rotation {pitch=-10, yaw=5}
    translation {y=5, z=-2}
   }
   n41 = Get thigh_rot_r()
   n42 = "Transform (Modify) Bone - Bone: dyn_skirt_fr_a_base_r"(ComponentPose=@n40, Alpha=@n41, Translation="-2,-5,0", Rotation="0,5,0", Scale="1,1,1") {
    rotationMode=BMM_Additive
    rotationSpace=BCS_BoneSpace
    scaleMode=BMM_Additive
    scaleSpace=BCS_BoneSpace
    translationMode=BMM_Additive
    translationSpace=BCS_BoneSpace
    alphaBoolBlend {blendOption=Linear}
    alphaScaleBiasClamp {bMapRange=true, inRange {max=0, min=-100}, outRange {max=0, min=1}}
    boneToModify {boneName=dyn_skirt_fr_a_base_r}
    rotation {yaw=5}
    translation=<-2 -5>
   }
   n43 = "Transform (Modify) Bone - Bone: dyn_skirt_side_aa_base_r"(ComponentPose=@n42, Alpha=@n41, Translation="0,-5,2", Rotation="-10,5,0", Scale="1,1,1") {
    rotationMode=BMM_Additive
    rotationSpace=BCS_BoneSpace
    scaleMode=BMM_Additive
    scaleSpace=BCS_BoneSpace
    translationMode=BMM_Additive
    translationSpace=BCS_BoneSpace
    alphaBoolBlend {blendOption=Linear}
    alphaScaleBiasClamp {bMapRange=true, inRange {max=0, min=-100}, outRange {max=0, min=1}}
    boneToModify {boneName=dyn_skirt_side_aa_base_r}
    rotation {pitch=-10, yaw=5}
    translation {y=-5, z=2}
   }
   n44 = RigidBody(ComponentPose=@n43, Alpha=0.7) {
    bClampLinearTranslationLimitToRefPose=true
    bOverrideWorldGravity=true
    overridePhysicsAsset=&/M_MED_Banana_Leader_Physics
    simulationSpace=BaseBoneSpace
    alphaBoolBlend {blendOption=Linear}
    baseBoneRef {boneName=root}
    componentLinearAccScale=<0.35 0.35 0.35>
    componentLinearVelScale=<0.35 0.35 0.35>
    overrideWorldGravity {z=-980}
   }
   Save cached pose 'Skirt_pose'(Pose=Component To Local(ComponentPose=@n44)) {cacheName=Skirt_pose}
   n45 = "Constraint - dyn_skirt_flap_aa_base_l (PPPPPP)"!(
    ComponentPose=Local To Component(LocalPose=Use cached pose 'Skirt_pose'())
    Alpha=1
    ConstraintWeights_0=0.5204
    ConstraintWeights_1:float=0.272
    ConstraintWeights_2:float=0.0303
    ConstraintWeights_3:float=0.0616
    ConstraintWeights_4:float=0.0991
    ConstraintWeights_5:float=0.0166
   ) -> (Pose) {
    alphaBoolBlend {blendOption=Linear}
    boneToModify {boneName=dyn_skirt_flap_aa_base_l}
    constraintSetup [
     {targetBone {boneName=dyn_skirt_fr_b_l}}
     {targetBone {boneName=dyn_skirt_fr_a_l}}
     {targetBone {boneName=dyn_skirt_side_ab_l}}
     {targetBone {boneName=dyn_skirt_fr_a_r}}
     {targetBone {boneName=dyn_skirt_fr_b_r}}
     {targetBone {boneName=dyn_skirt_side_aa_l}}
    ]
    constraintWeights [1,1,1,1,1,1]
   }
   n46 = "Constraint - dyn_skirt_flap_ab_l (PPPP)"!(
    ComponentPose=@n45
    Alpha=1
    ConstraintWeights_0=0.8204
    ConstraintWeights_1:float=0.1259
    ConstraintWeights_2:float=0.0185
    ConstraintWeights_3:float=0.0352
   ) -> (Pose) {
    alphaBoolBlend {blendOption=Linear}
    boneToModify {boneName=dyn_skirt_flap_ab_l}
    constraintSetup [
     {targetBone {boneName=dyn_skirt_fr_b_l}}
     {targetBone {boneName=dyn_skirt_fr_b_r}}
     {targetBone {boneName=dyn_skirt_fr_a_l}}
     {targetBone {boneName=dyn_skirt_side_ab_l}}
    ]
    constraintWeights [1,1,1,1]
   }
   n47 = "Constraint - dyn_skirt_flap_ba_base_l (PPPP)"!(
    ComponentPose=@n46
    Alpha=1
    ConstraintWeights_0=0.1852
    ConstraintWeights_1:float=0.2918
    ConstraintWeights_2:float=0.3709
    ConstraintWeights_3:float=0.1521
   ) -> (Pose) {
    alphaBoolBlend {blendOption=Linear}
    boneToModify {boneName=dyn_skirt_flap_ba_base_l}
    constraintSetup [
     {targetBone {boneName=dyn_skirt_side_aa_l}}
     {targetBone {boneName=dyn_skirt_fr_b_l}}
     {targetBone {boneName=dyn_skirt_fr_a_l}}
     {targetBone {boneName=dyn_skirt_side_ab_l}}
    ]
    constraintWeights [1,1,1,1]
   }
   n48 = "Constraint - dyn_skirt_flap_bb_l (PPPP)"!(
    ComponentPose=@n47
    Alpha=1
    ConstraintWeights_0=0.0224
    ConstraintWeights_1:float=0.6295
    ConstraintWeights_2:float=0.0411
    ConstraintWeights_3:float=0.3069
   ) -> (Pose) {
    alphaBoolBlend {blendOption=Linear}
    boneToModify {boneName=dyn_skirt_flap_bb_l}
    constraintSetup [
     {targetBone {boneName=dyn_skirt_side_aa_l}}
     {targetBone {boneName=dyn_skirt_fr_b_l}}
     {targetBone {boneName=dyn_skirt_fr_a_l}}
     {targetBone {boneName=dyn_skirt_side_ab_l}}
    ]
    constraintWeights [1,1,1,1]
   }
   n49 = "Constraint - dyn_skirt_flap_ca_base_l (PPPPPPP)"!(
    ComponentPose=@n48
    Alpha=1
    ConstraintWeights_0=0.092
    ConstraintWeights_1:float=0.0694
    ConstraintWeights_2:float=0.26
    ConstraintWeights_3:float=0.0972
    ConstraintWeights_4:float=0.4062
    ConstraintWeights_5:float=0.0634
    ConstraintWeights_6:float=0.0118
   ) -> (Pose) {
    alphaBoolBlend {blendOption=Linear}
    boneToModify {boneName=dyn_skirt_flap_ca_base_l}
    constraintSetup [
     {targetBone {boneName=dyn_skirt_side_ba_l}}
     {targetBone {boneName=dyn_skirt_fr_b_l}}
     {targetBone {boneName=dyn_skirt_fr_a_l}}
     {targetBone {boneName=dyn_skirt_side_ab_l}}
     {targetBone {boneName=dyn_skirt_side_aa_l}}
     {targetBone {boneName=pelvis}}
     {targetBone {boneName=dyn_skirt_side_bb_l}}
    ]
    constraintWeights [1,1,1,1,1,1,1]
   }
   n50 = "Constraint - dyn_skirt_flap_cb_l (PPPPP)"!(
    ComponentPose=@n49
    Alpha=1
    ConstraintWeights_0=0.0329
    ConstraintWeights_1:float=0.0499
    ConstraintWeights_2:float=0.3368
    ConstraintWeights_3:float=0.023
    ConstraintWeights_4:float=0.5574
   ) -> (Pose) {
    alphaBoolBlend {blendOption=Linear}
    boneToModify {boneName=dyn_skirt_flap_cb_l}
    constraintSetup [
     {targetBone {boneName=dyn_skirt_side_aa_l}}
     {targetBone {boneName=dyn_skirt_side_bb_l}}
     {targetBone {boneName=dyn_skirt_fr_b_l}}
     {targetBone {boneName=dyn_skirt_fr_a_l}}
     {targetBone {boneName=dyn_skirt_side_ab_l}}
    ]
    constraintWeights [1,1,1,1,1]
   }
   n51 = "Constraint - dyn_skirt_flap_da_base_l (PPPPPPP)"!(
    ComponentPose=@n50
    Alpha=1
    ConstraintWeights_0=0.0369
    ConstraintWeights_1:float=0.3687
    ConstraintWeights_2:float=0.1022
    ConstraintWeights_3:float=0.0359
    ConstraintWeights_4:float=0.0695
    ConstraintWeights_5:float=0.3056
    ConstraintWeights_6:float=0.0811
   ) -> (Pose) {
    alphaBoolBlend {blendOption=Linear}
    boneToModify {boneName=dyn_skirt_flap_da_base_l}
    constraintSetup [
     {targetBone {boneName=dyn_skirt_bk_a_l}}
     {targetBone {boneName=dyn_skirt_side_ba_l}}
     {targetBone {boneName=dyn_skirt_side_bb_l}}
     {targetBone {boneName=dyn_skirt_fr_a_l}}
     {targetBone {boneName=dyn_skirt_side_ab_l}}
     {targetBone {boneName=dyn_skirt_side_aa_l}}
     {targetBone {boneName=pelvis}}
    ]
    constraintWeights [1,1,1,1,1,1,1]
   }
   n52 = "Constraint - dyn_skirt_flap_db_l (PPPPPP)"!(
    ComponentPose=@n51
    Alpha=1
    ConstraintWeights_0=0.0312
    ConstraintWeights_1:float=0.4792
    ConstraintWeights_2:float=0.025
    ConstraintWeights_3:float=0.0208
    ConstraintWeights_4:float=0.4176
    ConstraintWeights_5:float=0.0263
   ) -> (Pose) {
    alphaBoolBlend {blendOption=Linear}
    boneToModify {boneName=dyn_skirt_flap_db_l}
    constraintSetup [
     {targetBone {boneName=dyn_skirt_side_ba_l}}
     {targetBone {boneName=dyn_skirt_side_bb_l}}
     {targetBone {boneName=dyn_skirt_bk_b_l}}
     {targetBone {boneName=dyn_skirt_side_aa_l}}
     {targetBone {boneName=dyn_skirt_side_ab_l}}
     {targetBone {boneName=dyn_skirt_fr_b_l}}
    ]
    constraintWeights [1,1,1,1,1,1]
   }
   n53 = "Constraint - dyn_skirt_flap_ea_base_l (PPPPPPP)"!(
    ComponentPose=@n52
    Alpha=1
    ConstraintWeights_0=0.1215
    ConstraintWeights_1:float=0.4778
    ConstraintWeights_2:float=0.1279
    ConstraintWeights_3:float=0.0302
    ConstraintWeights_4:float=0.1325
    ConstraintWeights_5:float=0.0829
    ConstraintWeights_6:float=0.0273
   ) -> (Pose) {
    alphaBoolBlend {blendOption=Linear}
    boneToModify {boneName=dyn_skirt_flap_ea_base_l}
    constraintSetup [
     {targetBone {boneName=dyn_skirt_bk_a_l}}
     {targetBone {boneName=dyn_skirt_side_ba_l}}
     {targetBone {boneName=dyn_skirt_side_bb_l}}
     {targetBone {boneName=dyn_skirt_bk_b_l}}
     {targetBone {boneName=dyn_skirt_side_aa_l}}
     {targetBone {boneName=pelvis}}
     {targetBone {boneName=dyn_skirt_side_ab_l}}
    ]
    constraintWeights [1,1,1,1,1,1,1]
   }
   n54 = "Constraint - dyn_skirt_flap_eb_l (PPPP)"!(
    ComponentPose=@n53
    Alpha=1
    ConstraintWeights_0=0.6895
    ConstraintWeights_1:float=0.1159
    ConstraintWeights_2:float=0.0402
    ConstraintWeights_3:float=0.1545
   ) -> (Pose) {
    alphaBoolBlend {blendOption=Linear}
    boneToModify {boneName=dyn_skirt_flap_eb_l}
    constraintSetup [
     {targetBone {boneName=dyn_skirt_side_bb_l}}
     {targetBone {boneName=dyn_skirt_bk_b_l}}
     {targetBone {boneName=dyn_skirt_side_ba_l}}
     {targetBone {boneName=dyn_skirt_side_ab_l}}
    ]
    constraintWeights [1,1,1,1]
   }
   n55 = "Constraint - dyn_skirt_flap_fa_base_l (PPPPPP)"!(
    ComponentPose=@n54
    Alpha=1
    ConstraintWeights_0=0.2142
    ConstraintWeights_1:float=0.1149
    ConstraintWeights_2:float=0.2179
    ConstraintWeights_3:float=0.419
    ConstraintWeights_4:float=0.0218
    ConstraintWeights_5:float=0.0122
   ) -> (Pose) {
    alphaBoolBlend {blendOption=Linear}
    boneToModify {boneName=dyn_skirt_flap_fa_base_l}
    constraintSetup [
     {targetBone {boneName=dyn_skirt_bk_a_l}}
     {targetBone {boneName=dyn_skirt_side_ba_l}}
     {targetBone {boneName=dyn_skirt_side_bb_l}}
     {targetBone {boneName=dyn_skirt_bk_b_l}}
     {targetBone {boneName=dyn_skirt_bk_b_r}}
     {targetBone {boneName=dyn_skirt_bk_a_r}}
    ]
    constraintWeights [1,1,1,1,1,1]
   }
   n56 = "Constraint - dyn_skirt_flap_fb_l (PPPPP)"!(
    ComponentPose=@n55
    Alpha=1
    ConstraintWeights_0=0.0356
    ConstraintWeights_1:float=0.0241
    ConstraintWeights_2:float=0.3018
    ConstraintWeights_3:float=0.0206
    ConstraintWeights_4:float=0.6179
   ) -> (Pose) {
    alphaBoolBlend {blendOption=Linear}
    boneToModify {boneName=dyn_skirt_flap_fb_l}
    constraintSetup [
     {targetBone {boneName=dyn_skirt_bk_a_l}}
     {targetBone {boneName=dyn_skirt_bk_b_r}}
     {targetBone {boneName=dyn_skirt_side_bb_l}}
     {targetBone {boneName=dyn_skirt_side_ba_l}}
     {targetBone {boneName=dyn_skirt_bk_b_l}}
    ]
    constraintWeights [1,1,1,1,1]
   }
   n57 = "Constraint - dyn_skirt_flap_ga_base_l (PPPPP)"!(
    ComponentPose=@n56
    Alpha=1
    ConstraintWeights_0=0.1068
    ConstraintWeights_1:float=0.251
    ConstraintWeights_2:float=0.0316
    ConstraintWeights_3:float=0.0511
    ConstraintWeights_4:float=0.5595
   ) -> (Pose) {
    alphaBoolBlend {blendOption=Linear}
    boneToModify {boneName=dyn_skirt_flap_ga_base_l}
    constraintSetup [
     {targetBone {boneName=dyn_skirt_bk_a_l}}
     {targetBone {boneName=dyn_skirt_bk_b_r}}
     {targetBone {boneName=dyn_skirt_side_bb_l}}
     {targetBone {boneName=dyn_skirt_bk_a_r}}
     {targetBone {boneName=dyn_skirt_bk_b_l}}
    ]
    constraintWeights [1,1,1,1,1]
   }
   n58 = "Constraint - dyn_skirt_flap_gb_l (PPP)"!(ComponentPose=@n57, Alpha=1, ConstraintWeights_0=0.248, ConstraintWeights_1:float=0.0173, ConstraintWeights_2:float=0.7347) -> (Pose) {
    alphaBoolBlend {blendOption=Linear}
    boneToModify {boneName=dyn_skirt_flap_gb_l}
    constraintSetup [{targetBone {boneName=dyn_skirt_bk_b_r}}, {targetBone {boneName=dyn_skirt_side_bb_l}}, {targetBone {boneName=dyn_skirt_bk_b_l}}]
    constraintWeights [1,1,1]
   }
   n59 = "Constraint - dyn_skirt_flap_aa_base_r (PPPPPP)"!(
    ComponentPose=@n58
    Alpha=1
    ConstraintWeights_0=0.0104
    ConstraintWeights_1:float=0.1721
    ConstraintWeights_2:float=0.1026
    ConstraintWeights_3:float=0.0189
    ConstraintWeights_4:float=0.2389
    ConstraintWeights_5:float=0.4572
   ) -> (Pose) {
    alphaBoolBlend {blendOption=Linear}
    boneToModify {boneName=dyn_skirt_flap_aa_base_r}
    constraintSetup [
     {targetBone {boneName=dyn_skirt_side_aa_r}}
     {targetBone {boneName=dyn_skirt_fr_b_l}}
     {targetBone {boneName=dyn_skirt_fr_a_l}}
     {targetBone {boneName=dyn_skirt_side_ab_r}}
     {targetBone {boneName=dyn_skirt_fr_a_r}}
     {targetBone {boneName=dyn_skirt_fr_b_r}}
    ]
    constraintWeights [1,1,1,1,1,1]
   }
   n60 = "Constraint - dyn_skirt_flap_ab_r (PPPPP)"!(
    ComponentPose=@n59
    Alpha=1
    ConstraintWeights_0=0.025
    ConstraintWeights_1:float=0.2579
    ConstraintWeights_2:float=0.0204
    ConstraintWeights_3:float=0.6863
    ConstraintWeights_4:float=0.0104
   ) -> (Pose) {
    alphaBoolBlend {blendOption=Linear}
    boneToModify {boneName=dyn_skirt_flap_ab_r}
    constraintSetup [
     {targetBone {boneName=dyn_skirt_side_ab_r}}
     {targetBone {boneName=dyn_skirt_fr_b_l}}
     {targetBone {boneName=dyn_skirt_fr_a_r}}
     {targetBone {boneName=dyn_skirt_fr_b_r}}
     {targetBone {boneName=dyn_skirt_fr_a_l}}
    ]
    constraintWeights [1,1,1,1,1]
   }
   n61 = "Constraint - dyn_skirt_flap_ba_base_r (PPPPP)"!(
    ComponentPose=@n60
    Alpha=1
    ConstraintWeights_0=0.106
    ConstraintWeights_1:float=0.4186
    ConstraintWeights_2:float=0.3254
    ConstraintWeights_3:float=0.0168
    ConstraintWeights_4:float=0.1332
   ) -> (Pose) {
    alphaBoolBlend {blendOption=Linear}
    boneToModify {boneName=dyn_skirt_flap_ba_base_r}
    constraintSetup [
     {targetBone {boneName=dyn_skirt_side_ab_r}}
     {targetBone {boneName=dyn_skirt_fr_a_r}}
     {targetBone {boneName=dyn_skirt_fr_b_r}}
     {targetBone {boneName=dyn_skirt_fr_a_l}}
     {targetBone {boneName=dyn_skirt_side_aa_r}}
    ]
    constraintWeights [1,1,1,1,1]
   }
   n62 = "Constraint - dyn_skirt_flap_bb_r (PPPPP)"!(
    ComponentPose=@n61
    Alpha=1
    ConstraintWeights_0=0.2014
    ConstraintWeights_1:float=0.0165
    ConstraintWeights_2:float=0.0421
    ConstraintWeights_3:float=0.7262
    ConstraintWeights_4:float=0.0139
   ) -> (Pose) {
    alphaBoolBlend {blendOption=Linear}
    boneToModify {boneName=dyn_skirt_flap_bb_r}
    constraintSetup [
     {targetBone {boneName=dyn_skirt_side_ab_r}}
     {targetBone {boneName=dyn_skirt_fr_b_l}}
     {targetBone {boneName=dyn_skirt_fr_a_r}}
     {targetBone {boneName=dyn_skirt_fr_b_r}}
     {targetBone {boneName=dyn_skirt_side_aa_r}}
    ]
    constraintWeights [1,1,1,1,1]
   }
   n63 = "Constraint - dyn_skirt_flap_ca_base_r (PPPPPPP)"!(
    ComponentPose=@n62
    Alpha=1
    ConstraintWeights_0=0.4119
    ConstraintWeights_1:float=0.1022
    ConstraintWeights_2:float=0.189
    ConstraintWeights_3:float=0.0501
    ConstraintWeights_4:float=0.0278
    ConstraintWeights_5:float=0.0666
    ConstraintWeights_6:float=0.1524
   ) -> (Pose) {
    alphaBoolBlend {blendOption=Linear}
    boneToModify {boneName=dyn_skirt_flap_ca_base_r}
    constraintSetup [
     {targetBone {boneName=dyn_skirt_side_aa_r}}
     {targetBone {boneName=dyn_skirt_side_ab_r}}
     {targetBone {boneName=dyn_skirt_fr_a_r}}
     {targetBone {boneName=dyn_skirt_fr_b_r}}
     {targetBone {boneName=dyn_skirt_side_bb_r}}
     {targetBone {boneName=pelvis}}
     {targetBone {boneName=dyn_skirt_side_ba_r}}
    ]
    constraintWeights [1,1,1,1,1,1,1]
   }
   n64 = "Constraint - dyn_skirt_flap_cb_r (PPPPP)"!(
    ComponentPose=@n63
    Alpha=1
    ConstraintWeights_0=0.6041
    ConstraintWeights_1:float=0.0117
    ConstraintWeights_2:float=0.2312
    ConstraintWeights_3:float=0.1203
    ConstraintWeights_4:float=0.0327
   ) -> (Pose) {
    alphaBoolBlend {blendOption=Linear}
    boneToModify {boneName=dyn_skirt_flap_cb_r}
    constraintSetup [
     {targetBone {boneName=dyn_skirt_side_ab_r}}
     {targetBone {boneName=dyn_skirt_fr_a_r}}
     {targetBone {boneName=dyn_skirt_fr_b_r}}
     {targetBone {boneName=dyn_skirt_side_bb_r}}
     {targetBone {boneName=dyn_skirt_side_aa_r}}
    ]
    constraintWeights [1,1,1,1,1]
   }
   n65 = "Constraint - dyn_skirt_flap_da_base_r (PPPPPPPP)"!(
    ComponentPose=@n64
    Alpha=1
    ConstraintWeights_0=0.3623
    ConstraintWeights_1:float=0.0883
    ConstraintWeights_2:float=0.0758
    ConstraintWeights_3:float=0.0148
    ConstraintWeights_4:float=0.0734
    ConstraintWeights_5:float=0.0756
    ConstraintWeights_6:float=0.2915
    ConstraintWeights_7:float=0.0183
   ) -> (Pose) {
    alphaBoolBlend {blendOption=Linear}
    boneToModify {boneName=dyn_skirt_flap_da_base_r}
    constraintSetup [
     {targetBone {boneName=dyn_skirt_side_aa_r}}
     {targetBone {boneName=dyn_skirt_side_ab_r}}
     {targetBone {boneName=dyn_skirt_fr_a_r}}
     {targetBone {boneName=dyn_skirt_fr_b_r}}
     {targetBone {boneName=dyn_skirt_side_bb_r}}
     {targetBone {boneName=pelvis}}
     {targetBone {boneName=dyn_skirt_side_ba_r}}
     {targetBone {boneName=dyn_skirt_bk_a_r}}
    ]
    constraintWeights [1,1,1,1,1,1,1,1]
   }
   n66 = "Constraint - dyn_skirt_flap_db_r (PPPPPP)"!(
    ComponentPose=@n65
    Alpha=1
    ConstraintWeights_0=0.0263
    ConstraintWeights_1:float=0.0125
    ConstraintWeights_2:float=0.5308
    ConstraintWeights_3:float=0.0709
    ConstraintWeights_4:float=0.3381
    ConstraintWeights_5:float=0.0214
   ) -> (Pose) {
    alphaBoolBlend {blendOption=Linear}
    boneToModify {boneName=dyn_skirt_flap_db_r}
    constraintSetup [
     {targetBone {boneName=dyn_skirt_side_aa_r}}
     {targetBone {boneName=dyn_skirt_bk_b_r}}
     {targetBone {boneName=dyn_skirt_side_ab_r}}
     {targetBone {boneName=dyn_skirt_fr_b_r}}
     {targetBone {boneName=dyn_skirt_side_bb_r}}
     {targetBone {boneName=dyn_skirt_side_ba_r}}
    ]
    constraintWeights [1,1,1,1,1,1]
   }
   n67 = "Constraint - dyn_skirt_flap_ea_base_r (PPPPPPP)"!(
    ComponentPose=@n66
    Alpha=1
    ConstraintWeights_0=0.0845
    ConstraintWeights_1:float=0.0129
    ConstraintWeights_2:float=0.0533
    ConstraintWeights_3:float=0.1197
    ConstraintWeights_4:float=0.084
    ConstraintWeights_5:float=0.44
    ConstraintWeights_6:float=0.2056
   ) -> (Pose) {
    alphaBoolBlend {blendOption=Linear}
    boneToModify {boneName=dyn_skirt_flap_ea_base_r}
    constraintSetup [
     {targetBone {boneName=dyn_skirt_side_aa_r}}
     {targetBone {boneName=dyn_skirt_side_ab_r}}
     {targetBone {boneName=dyn_skirt_bk_b_r}}
     {targetBone {boneName=dyn_skirt_side_bb_r}}
     {targetBone {boneName=pelvis}}
     {targetBone {boneName=dyn_skirt_side_ba_r}}
     {targetBone {boneName=dyn_skirt_bk_a_r}}
    ]
    constraintWeights [1,1,1,1,1,1,1]
   }
   n68 = "Constraint - dyn_skirt_flap_eb_r (PPPPP)"!(
    ComponentPose=@n67
    Alpha=1
    ConstraintWeights_0=0.0877
    ConstraintWeights_1:float=0.0369
    ConstraintWeights_2:float=0.6235
    ConstraintWeights_3:float=0.0127
    ConstraintWeights_4:float=0.2392
   ) -> (Pose) {
    alphaBoolBlend {blendOption=Linear}
    boneToModify {boneName=dyn_skirt_flap_eb_r}
    constraintSetup [
     {targetBone {boneName=dyn_skirt_side_ab_r}}
     {targetBone {boneName=dyn_skirt_side_ba_r}}
     {targetBone {boneName=dyn_skirt_side_bb_r}}
     {targetBone {boneName=dyn_skirt_bk_a_r}}
     {targetBone {boneName=dyn_skirt_bk_b_r}}
    ]
    constraintWeights [1,1,1,1,1]
   }
   n69 = "Constraint - dyn_skirt_flap_fa_base_r (PPPPPP)"!(
    ComponentPose=@n68
    Alpha=1
    ConstraintWeights_0=0.3436
    ConstraintWeights_1:float=0.012
    ConstraintWeights_2:float=0.0119
    ConstraintWeights_3:float=0.3004
    ConstraintWeights_4:float=0.1536
    ConstraintWeights_5:float=0.1784
   ) -> (Pose) {
    alphaBoolBlend {blendOption=Linear}
    boneToModify {boneName=dyn_skirt_flap_fa_base_r}
    constraintSetup [
     {targetBone {boneName=dyn_skirt_bk_b_r}}
     {targetBone {boneName=dyn_skirt_bk_b_l}}
     {targetBone {boneName=dyn_skirt_side_ab_r}}
     {targetBone {boneName=dyn_skirt_side_bb_r}}
     {targetBone {boneName=dyn_skirt_side_ba_r}}
     {targetBone {boneName=dyn_skirt_bk_a_r}}
    ]
    constraintWeights [1,1,1,1,1,1]
   }
   n70 = "Constraint - dyn_skirt_flap_fb_r (PPPPPP)"!(
    ComponentPose=@n69
    Alpha=1
    ConstraintWeights_0=0.493
    ConstraintWeights_1:float=0.0134
    ConstraintWeights_2:float=0.0158
    ConstraintWeights_3:float=0.4208
    ConstraintWeights_4:float=0.0271
    ConstraintWeights_5:float=0.0299
   ) -> (Pose) {
    alphaBoolBlend {blendOption=Linear}
    boneToModify {boneName=dyn_skirt_flap_fb_r}
    constraintSetup [
     {targetBone {boneName=dyn_skirt_bk_b_r}}
     {targetBone {boneName=dyn_skirt_bk_b_l}}
     {targetBone {boneName=dyn_skirt_side_ab_r}}
     {targetBone {boneName=dyn_skirt_side_bb_r}}
     {targetBone {boneName=dyn_skirt_side_ba_r}}
     {targetBone {boneName=dyn_skirt_bk_a_r}}
    ]
    constraintWeights [1,1,1,1,1,1]
   }
   n71 = "Constraint - dyn_skirt_flap_ga_base_r (PPPPPP)"!(
    ComponentPose=@n70
    Alpha=1
    ConstraintWeights_0=0.037
    ConstraintWeights_1:float=0.1784
    ConstraintWeights_2:float=0.595
    ConstraintWeights_3:float=0.0653
    ConstraintWeights_4:float=0.011
    ConstraintWeights_5:float=0.1133
   ) -> (Pose) {
    alphaBoolBlend {blendOption=Linear}
    boneToModify {boneName=dyn_skirt_flap_ga_base_r}
    constraintSetup [
     {targetBone {boneName=dyn_skirt_bk_a_l}}
     {targetBone {boneName=dyn_skirt_bk_b_l}}
     {targetBone {boneName=dyn_skirt_bk_b_r}}
     {targetBone {boneName=dyn_skirt_side_bb_r}}
     {targetBone {boneName=dyn_skirt_side_ba_r}}
     {targetBone {boneName=dyn_skirt_bk_a_r}}
    ]
    constraintWeights [1,1,1,1,1,1]
   }
   n72 = "Constraint - dyn_skirt_flap_gb_r (PPP)"!(ComponentPose=@n71, Alpha=1, ConstraintWeights_0=0.7843, ConstraintWeights_1:float=0.0518, ConstraintWeights_2:float=0.1639) -> (Pose) {
    alphaBoolBlend {blendOption=Linear}
    boneToModify {boneName=dyn_skirt_flap_gb_r}
    constraintSetup [{targetBone {boneName=dyn_skirt_bk_b_r}}, {targetBone {boneName=dyn_skirt_side_bb_r}}, {targetBone {boneName=dyn_skirt_bk_b_l}}]
    constraintWeights [1,1,1]
   }
   n73 = Get thigh_rot_l()
   n74 = "Transform (Modify) Bone - Bone: dyn_skirt_flap_aa_base_l"(ComponentPose=@n72, Alpha=@n73, Translation="0,0,0", Rotation="0,0,40", Scale="1,1,1") {
    rotationMode=BMM_Additive
    rotationSpace=BCS_BoneSpace
    scaleMode=BMM_Additive
    scaleSpace=BCS_BoneSpace
    translationMode=BMM_Additive
    translationSpace=BCS_BoneSpace
    alphaBoolBlend {blendOption=Linear}
    alphaScaleBiasClamp {bMapRange=true, inRange {max=0, min=-100}, outRange {max=0, min=1}}
    boneToModify {boneName=dyn_skirt_flap_aa_base_l}
    rotation {roll=40}
   }
   n75 = "Transform (Modify) Bone - Bone: dyn_skirt_flap_ba_base_l"(ComponentPose=@n74, Alpha=@n73, Translation="0,0,0", Rotation="0,0,30", Scale="1,1,1") {
    rotationMode=BMM_Additive
    rotationSpace=BCS_BoneSpace
    scaleMode=BMM_Additive
    scaleSpace=BCS_BoneSpace
    translationMode=BMM_Additive
    translationSpace=BCS_BoneSpace
    alphaBoolBlend {blendOption=Linear}
    alphaScaleBiasClamp {bMapRange=true, inRange {max=0, min=-100}, outRange {max=0, min=1}}
    boneToModify {boneName=dyn_skirt_flap_ba_base_l}
    rotation {roll=30}
   }
   n76 = "Transform (Modify) Bone - Bone: dyn_skirt_flap_ca_base_l"(ComponentPose=@n75, Alpha=@n73, Translation="0,1,0", Rotation="5,0,0", Scale="1,1,1") {
    rotationMode=BMM_Additive
    rotationSpace=BCS_BoneSpace
    scaleMode=BMM_Additive
    scaleSpace=BCS_BoneSpace
    translationMode=BMM_Additive
    translationSpace=BCS_BoneSpace
    alphaBoolBlend {blendOption=Linear}
    alphaScaleBiasClamp {bMapRange=true, inRange {max=0, min=-100}, outRange {max=0, min=1}}
    boneToModify {boneName=dyn_skirt_flap_ca_base_l}
    rotation {pitch=5}
    translation {y=1}
   }
   n77 = "Transform (Modify) Bone - Bone: dyn_skirt_flap_aa_base_r"(ComponentPose=@n76, Alpha=@n73, Translation="0,0.5,0", Rotation="0,0,-30", Scale="1,1,1") {
    rotationMode=BMM_Additive
    rotationSpace=BCS_BoneSpace
    scaleMode=BMM_Additive
    scaleSpace=BCS_BoneSpace
    translationMode=BMM_Additive
    translationSpace=BCS_BoneSpace
    alphaBoolBlend {blendOption=Linear}
    alphaScaleBiasClamp {bMapRange=true, inRange {max=0, min=-100}, outRange {max=0, min=1}}
    boneToModify {boneName=dyn_skirt_flap_aa_base_r}
    rotation {roll=-30}
    translation {y=0.5}
   }
   n78 = Get thigh_rot_r()
   n79 = "Transform (Modify) Bone - Bone: dyn_skirt_flap_aa_base_r"(ComponentPose=@n77, Alpha=@n78, Translation="0,0,0", Rotation="-10,10,60", Scale="1,1,1") {
    rotationMode=BMM_Additive
    rotationSpace=BCS_BoneSpace
    scaleMode=BMM_Additive
    scaleSpace=BCS_BoneSpace
    translationMode=BMM_Additive
    translationSpace=BCS_BoneSpace
    alphaBoolBlend {blendOption=Linear}
    alphaScaleBiasClamp {bMapRange=true, inRange {max=0, min=-100}, outRange {max=0, min=1}}
    boneToModify {boneName=dyn_skirt_flap_aa_base_r}
    rotation=<rot -10 10 60>
   }
   n80 = "Transform (Modify) Bone - Bone: dyn_skirt_flap_ba_base_r"(ComponentPose=@n79, Alpha=@n78, Translation="0,0,0", Rotation="0,0,30", Scale="1,1,1") {
    rotationMode=BMM_Additive
    rotationSpace=BCS_BoneSpace
    scaleMode=BMM_Additive
    scaleSpace=BCS_BoneSpace
    translationMode=BMM_Additive
    translationSpace=BCS_BoneSpace
    alphaBoolBlend {blendOption=Linear}
    alphaScaleBiasClamp {bMapRange=true, inRange {max=0, min=-100}, outRange {max=0, min=1}}
    boneToModify {boneName=dyn_skirt_flap_ba_base_r}
    rotation {roll=30}
   }
   n81 = "Transform (Modify) Bone - Bone: dyn_skirt_flap_ca_base_r"(ComponentPose=@n80, Alpha=@n78, Translation="0,-1,0", Rotation="5,0,0", Scale="1,1,1") {
    rotationMode=BMM_Additive
    rotationSpace=BCS_BoneSpace
    scaleMode=BMM_Additive
    scaleSpace=BCS_BoneSpace
    translationMode=BMM_Additive
    translationSpace=BCS_BoneSpace
    alphaBoolBlend {blendOption=Linear}
    alphaScaleBiasClamp {bMapRange=true, inRange {max=0, min=-100}, outRange {max=0, min=1}}
    boneToModify {boneName=dyn_skirt_flap_ca_base_r}
    rotation {pitch=5}
    translation {y=-1}
   }
   n82 = "Transform (Modify) Bone - Bone: dyn_skirt_flap_aa_base_l"(ComponentPose=@n81, Alpha=@n78, Translation="0,0,0", Rotation="0,0,-30", Scale="1,1,1") {
    rotationMode=BMM_Additive
    rotationSpace=BCS_BoneSpace
    scaleMode=BMM_Additive
    scaleSpace=BCS_BoneSpace
    translationMode=BMM_Additive
    translationSpace=BCS_BoneSpace
    alphaBoolBlend {blendOption=Linear}
    alphaScaleBiasClamp {bMapRange=true, inRange {max=0, min=-100}, outRange {max=0, min=1}}
    boneToModify {boneName=dyn_skirt_flap_aa_base_l}
    rotation {roll=-30}
   }
   n83 = "Transform (Modify) Bone - Bone: dyn_skirt_flap_fa_base_l"(ComponentPose=@n82, Alpha=1, Translation="0,1,0", Rotation="0,0,0", Scale="1,1,1") {
    rotationMode=BMM_Additive
    rotationSpace=BCS_BoneSpace
    scaleMode=BMM_Additive
    scaleSpace=BCS_BoneSpace
    translationMode=BMM_Additive
    translationSpace=BCS_BoneSpace
    alphaBoolBlend {blendOption=Linear}
    alphaScaleBiasClamp {inRange {max=0, min=-100}, outRange {max=0, min=1}}
    boneToModify {boneName=dyn_skirt_flap_fa_base_l}
    translation {y=1}
   }
   n84 = "Transform (Modify) Bone - Bone: dyn_skirt_flap_ga_base_l"(ComponentPose=@n83, Alpha=1, Translation="0,1,0", Rotation="0,0,0", Scale="1,1,1") {
    rotationMode=BMM_Additive
    rotationSpace=BCS_BoneSpace
    scaleMode=BMM_Additive
    scaleSpace=BCS_BoneSpace
    translationMode=BMM_Additive
    translationSpace=BCS_BoneSpace
    alphaBoolBlend {blendOption=Linear}
    alphaScaleBiasClamp {inRange {max=0, min=-100}, outRange {max=0, min=1}}
    boneToModify {boneName=dyn_skirt_flap_ga_base_l}
    translation {y=1}
   }
   n85 = "Transform (Modify) Bone - Bone: dyn_skirt_flap_fa_base_r"(ComponentPose=@n84, Alpha=1, Translation="0,-1,0", Rotation="0,0,0", Scale="1,1,1") {
    rotationMode=BMM_Additive
    rotationSpace=BCS_BoneSpace
    scaleMode=BMM_Additive
    scaleSpace=BCS_BoneSpace
    translationMode=BMM_Additive
    translationSpace=BCS_BoneSpace
    alphaBoolBlend {blendOption=Linear}
    alphaScaleBiasClamp {inRange {max=0, min=-100}, outRange {max=0, min=1}}
    boneToModify {boneName=dyn_skirt_flap_fa_base_r}
    translation {y=-1}
   }
   n86 = "Transform (Modify) Bone - Bone: dyn_skirt_flap_ga_base_r"(ComponentPose=@n85, Alpha=1, Translation="0,-1,0", Rotation="0,0,0", Scale="1,1,1") {
    rotationMode=BMM_Additive
    rotationSpace=BCS_BoneSpace
    scaleMode=BMM_Additive
    scaleSpace=BCS_BoneSpace
    translationMode=BMM_Additive
    translationSpace=BCS_BoneSpace
    alphaBoolBlend {blendOption=Linear}
    alphaScaleBiasClamp {inRange {max=0, min=-100}, outRange {max=0, min=1}}
    boneToModify {boneName=dyn_skirt_flap_ga_base_r}
    translation {y=-1}
   }
   n87 = Get jumpingAlpha() -> (jumpingAlpha:double)
   n88 = "Transform (Modify) Bone - Bone: dyn_skirt_flap_ba_base_r"(ComponentPose=@n86, Alpha=@n87, Translation="0,0,0", Rotation="0,10,0", Scale="1,1,1") {
    rotationMode=BMM_Additive
    rotationSpace=BCS_BoneSpace
    scaleMode=BMM_Additive
    scaleSpace=BCS_BoneSpace
    translationMode=BMM_Additive
    translationSpace=BCS_BoneSpace
    alphaBoolBlend {blendOption=Linear}
    alphaScaleBiasClamp {inRange {max=0, min=-100}, outRange {max=0, min=1}}
    boneToModify {boneName=dyn_skirt_flap_ba_base_r}
    rotation {yaw=10}
   }
   n89 = "Transform (Modify) Bone - Bone: dyn_skirt_flap_ca_base_r"(ComponentPose=@n88, Alpha=@n87, Translation="0,0,0", Rotation="0,10,0", Scale="1,1,1") {
    rotationMode=BMM_Additive
    rotationSpace=BCS_BoneSpace
    scaleMode=BMM_Additive
    scaleSpace=BCS_BoneSpace
    translationMode=BMM_Additive
    translationSpace=BCS_BoneSpace
    alphaBoolBlend {blendOption=Linear}
    alphaScaleBiasClamp {inRange {max=0, min=-100}, outRange {max=0, min=1}}
    boneToModify {boneName=dyn_skirt_flap_ca_base_r}
    rotation {yaw=10}
   }
   n90 = "Transform (Modify) Bone - Bone: dyn_skirt_flap_da_base_r"(ComponentPose=@n89, Alpha=@n87, Translation="0,0,0", Rotation="0,10,0", Scale="1,1,1") {
    rotationMode=BMM_Additive
    rotationSpace=BCS_BoneSpace
    scaleMode=BMM_Additive
    scaleSpace=BCS_BoneSpace
    translationMode=BMM_Additive
    translationSpace=BCS_BoneSpace
    alphaBoolBlend {blendOption=Linear}
    alphaScaleBiasClamp {inRange {max=0, min=-100}, outRange {max=0, min=1}}
    boneToModify {boneName=dyn_skirt_flap_da_base_r}
    rotation {yaw=10}
   }
   n91 = "Transform (Modify) Bone - Bone: dyn_skirt_flap_ea_base_r"(ComponentPose=@n90, Alpha=@n87, Translation="0,0,0", Rotation="0,10,0", Scale="1,1,1") {
    rotationMode=BMM_Additive
    rotationSpace=BCS_BoneSpace
    scaleMode=BMM_Additive
    scaleSpace=BCS_BoneSpace
    translationMode=BMM_Additive
    translationSpace=BCS_BoneSpace
    alphaBoolBlend {blendOption=Linear}
    alphaScaleBiasClamp {inRange {max=0, min=-100}, outRange {max=0, min=1}}
    boneToModify {boneName=dyn_skirt_flap_ea_base_r}
    rotation {yaw=10}
   }
   n92 = "Transform (Modify) Bone - Bone: dyn_skirt_flap_ca_base_l"!(ComponentPose=@n91, bAlphaBoolEnabled:bool=Get bIsInFrontEnd(), Translation="0,0.5,0", Rotation="0,0,25", Scale="1,1,1") -> (Pose) {
    alphaInputType=Bool
    rotationMode=BMM_Additive
    rotationSpace=BCS_BoneSpace
    scaleMode=BMM_Additive
    scaleSpace=BCS_BoneSpace
    translationMode=BMM_Additive
    translationSpace=BCS_BoneSpace
    alphaBoolBlend {blendOption=Linear}
    alphaScaleBiasClamp {inRange {max=0, min=-100}, outRange {max=0, min=1}}
    boneToModify {boneName=dyn_skirt_flap_ca_base_l}
    rotation {roll=25}
    translation {y=0.5}
   }
   n93 = Get isCrouchingAlpha()
   n94 = Anim Dynamics - Bone: dyn_skirt_flap_aa_l(ComponentPose=@n92, Alpha=@n93, ExternalForce="0,0,0") {
    angularDampingOverride=1
    angularSpringConstant=1500
    bAngularSpring=true
    bGravityOverrideInSimSpace=true
    bOverrideAngularDamping=true
    gravityScale=0.35
    simulationSpace=RootRelative
    alphaBoolBlend {blendOption=Linear}
    alphaScaleBiasClamp {bMapRange=true, outRange {max=0.2, min=1}}
    boundBone {boneName=dyn_skirt_flap_aa_l}
    physicsBodyDefinitions [{
      sphereCollisionRadius=0
      boundBone {boneName=dyn_skirt_flap_aa_l}
      boxExtents=<10 20 20>
      constraintSetup {angularLimitsMax=<0 2.5 20>, angularLimitsMin=<0 -2.5 0>, angularTarget=<0 0 0>}
      localJointOffset=<10 0 0>
     }]
   }
   n95 = Anim Dynamics - Bone: dyn_skirt_flap_ab_l(ComponentPose=@n94, Alpha=@n93, ExternalForce="0,0,0") {
    angularDampingOverride=1
    angularSpringConstant=1500
    bAngularSpring=true
    bGravityOverrideInSimSpace=true
    bOverrideAngularDamping=true
    gravityScale=0.35
    simulationSpace=RootRelative
    alphaBoolBlend {blendOption=Linear}
    alphaScaleBiasClamp {bMapRange=true, outRange {max=0.2, min=1}}
    boundBone {boneName=dyn_skirt_flap_ab_l}
    physicsBodyDefinitions [{
      sphereCollisionRadius=0
      boundBone {boneName=dyn_skirt_flap_ab_l}
      boxExtents=<10 20 20>
      constraintSetup {angularLimitsMax=<0 2.5 20>, angularLimitsMin=<0 -2.5 0>, angularTarget=<0 0 0>}
      localJointOffset=<10 0 0>
     }]
   }
   n96 = Anim Dynamics - Bone: dyn_skirt_flap_ba_l(ComponentPose=@n95, Alpha=@n93, ExternalForce="0,0,0") {
    angularDampingOverride=1
    angularSpringConstant=1500
    bAngularSpring=true
    bGravityOverrideInSimSpace=true
    bOverrideAngularDamping=true
    gravityScale=0.35
    simulationSpace=RootRelative
    alphaBoolBlend {blendOption=Linear}
    alphaScaleBiasClamp {bMapRange=true, outRange {max=0.2, min=1}}
    boundBone {boneName=dyn_skirt_flap_ba_l}
    physicsBodyDefinitions [{
      sphereCollisionRadius=0
      boundBone {boneName=dyn_skirt_flap_ba_l}
      boxExtents=<10 20 20>
      constraintSetup {angularLimitsMax=<0 2.5 20>, angularLimitsMin=<0 -2.5 0>, angularTarget=<0 0 0>}
      localJointOffset=<10 0 0>
     }]
   }
   n97 = Anim Dynamics - Bone: dyn_skirt_flap_bb_l(ComponentPose=@n96, Alpha=@n93, ExternalForce="0,0,0") {
    angularDampingOverride=1
    angularSpringConstant=1500
    bAngularSpring=true
    bGravityOverrideInSimSpace=true
    bOverrideAngularDamping=true
    gravityScale=0.35
    simulationSpace=RootRelative
    alphaBoolBlend {blendOption=Linear}
    alphaScaleBiasClamp {bMapRange=true, outRange {max=0.2, min=1}}
    boundBone {boneName=dyn_skirt_flap_bb_l}
    physicsBodyDefinitions [{
      sphereCollisionRadius=0
      boundBone {boneName=dyn_skirt_flap_bb_l}
      boxExtents=<10 20 20>
      constraintSetup {angularLimitsMax=<0 2.5 20>, angularLimitsMin=<0 -2.5 0>, angularTarget=<0 0 0>}
      localJointOffset=<10 0 0>
     }]
   }
   n98 = Anim Dynamics - Bone: dyn_skirt_flap_ca_l(ComponentPose=@n97, Alpha=@n93, ExternalForce="0,0,0") {
    angularDampingOverride=1
    angularSpringConstant=1500
    bAngularSpring=true
    bGravityOverrideInSimSpace=true
    bOverrideAngularDamping=true
    gravityScale=0.35
    simulationSpace=RootRelative
    alphaBoolBlend {blendOption=Linear}
    alphaScaleBiasClamp {bMapRange=true, outRange {max=0.2, min=1}}
    boundBone {boneName=dyn_skirt_flap_ca_l}
    physicsBodyDefinitions [{
      sphereCollisionRadius=0
      boundBone {boneName=dyn_skirt_flap_ca_l}
      boxExtents=<10 20 20>
      constraintSetup {angularLimitsMax=<0 2.5 20>, angularLimitsMin=<0 -2.5 0>, angularTarget=<0 0 0>}
      localJointOffset=<10 0 0>
     }]
   }
   n99 = Anim Dynamics - Bone: dyn_skirt_flap_cb_l(ComponentPose=@n98, Alpha=@n93, ExternalForce="0,0,0") {
    angularDampingOverride=1
    angularSpringConstant=1500
    bAngularSpring=true
    bGravityOverrideInSimSpace=true
    bOverrideAngularDamping=true
    gravityScale=0.35
    simulationSpace=RootRelative
    alphaBoolBlend {blendOption=Linear}
    alphaScaleBiasClamp {bMapRange=true, outRange {max=0.2, min=1}}
    boundBone {boneName=dyn_skirt_flap_cb_l}
    physicsBodyDefinitions [{
      sphereCollisionRadius=0
      boundBone {boneName=dyn_skirt_flap_cb_l}
      boxExtents=<10 20 20>
      constraintSetup {angularLimitsMax=<0 2.5 20>, angularLimitsMin=<0 -2.5 0>, angularTarget=<0 0 0>}
      localJointOffset=<10 0 0>
     }]
   }
   n100 = Anim Dynamics - Bone: dyn_skirt_flap_da_l(ComponentPose=@n99, Alpha=@n93, ExternalForce="0,0,0") {
    angularDampingOverride=1
    angularSpringConstant=1500
    bAngularSpring=true
    bGravityOverrideInSimSpace=true
    bOverrideAngularDamping=true
    gravityScale=0.35
    simulationSpace=RootRelative
    alphaBoolBlend {blendOption=Linear}
    alphaScaleBiasClamp {bMapRange=true, outRange {max=0.2, min=1}}
    boundBone {boneName=dyn_skirt_flap_da_l}
    physicsBodyDefinitions [{
      sphereCollisionRadius=0
      boundBone {boneName=dyn_skirt_flap_da_l}
      boxExtents=<10 20 20>
      constraintSetup {angularLimitsMax=<0 2.5 20>, angularLimitsMin=<0 -2.5 0>, angularTarget=<0 0 0>}
      localJointOffset=<10 0 0>
     }]
   }
   n101 = Anim Dynamics - Bone: dyn_skirt_flap_db_l(ComponentPose=@n100, Alpha=@n93, ExternalForce="0,0,0") {
    angularDampingOverride=1
    angularSpringConstant=1500
    bAngularSpring=true
    bGravityOverrideInSimSpace=true
    bOverrideAngularDamping=true
    gravityScale=0.35
    simulationSpace=RootRelative
    alphaBoolBlend {blendOption=Linear}
    alphaScaleBiasClamp {bMapRange=true, outRange {max=0.2, min=1}}
    boundBone {boneName=dyn_skirt_flap_db_l}
    physicsBodyDefinitions [{
      sphereCollisionRadius=0
      boundBone {boneName=dyn_skirt_flap_db_l}
      boxExtents=<10 20 20>
      constraintSetup {angularLimitsMax=<0 2.5 20>, angularLimitsMin=<0 -2.5 0>, angularTarget=<0 0 0>}
      localJointOffset=<10 0 0>
     }]
   }
   n102 = Anim Dynamics - Bone: dyn_skirt_flap_ea_l(ComponentPose=@n101, Alpha=1, ExternalForce="0,0,0") {
    angularDampingOverride=1
    angularSpringConstant=1500
    bAngularSpring=true
    bGravityOverrideInSimSpace=true
    bOverrideAngularDamping=true
    gravityScale=0.35
    simulationSpace=RootRelative
    alphaBoolBlend {blendOption=Linear}
    boundBone {boneName=dyn_skirt_flap_ea_l}
    physicsBodyDefinitions [{
      sphereCollisionRadius=0
      boundBone {boneName=dyn_skirt_flap_ea_l}
      boxExtents=<10 20 20>
      constraintSetup {angularLimitsMax=<0 2.5 20>, angularLimitsMin=<0 -2.5 0>, angularTarget=<0 0 0>}
      localJointOffset=<10 0 0>
     }]
   }
   n103 = Anim Dynamics - Bone: dyn_skirt_flap_eb_l(ComponentPose=@n102, Alpha=1, ExternalForce="0,0,0") {
    angularDampingOverride=1
    angularSpringConstant=1500
    bAngularSpring=true
    bGravityOverrideInSimSpace=true
    bOverrideAngularDamping=true
    gravityScale=0.35
    simulationSpace=RootRelative
    alphaBoolBlend {blendOption=Linear}
    boundBone {boneName=dyn_skirt_flap_eb_l}
    physicsBodyDefinitions [{
      sphereCollisionRadius=0
      boundBone {boneName=dyn_skirt_flap_eb_l}
      boxExtents=<10 20 20>
      constraintSetup {angularLimitsMax=<0 2.5 20>, angularLimitsMin=<0 -2.5 0>, angularTarget=<0 0 0>}
      localJointOffset=<10 0 0>
     }]
   }
   n104 = Anim Dynamics - Bone: dyn_skirt_flap_fa_l(ComponentPose=@n103, Alpha=1, ExternalForce="0,0,0") {
    angularDampingOverride=1
    angularSpringConstant=1500
    bAngularSpring=true
    bGravityOverrideInSimSpace=true
    bOverrideAngularDamping=true
    gravityScale=0.35
    simulationSpace=RootRelative
    alphaBoolBlend {blendOption=Linear}
    boundBone {boneName=dyn_skirt_flap_fa_l}
    physicsBodyDefinitions [{
      sphereCollisionRadius=0
      boundBone {boneName=dyn_skirt_flap_fa_l}
      boxExtents=<10 20 20>
      constraintSetup {angularLimitsMax=<0 2.5 20>, angularLimitsMin=<0 -2.5 0>, angularTarget=<0 0 0>}
      localJointOffset=<10 0 0>
     }]
   }
   n105 = Anim Dynamics - Bone: dyn_skirt_flap_fb_l(ComponentPose=@n104, Alpha=1, ExternalForce="0,0,0") {
    angularDampingOverride=1
    angularSpringConstant=1500
    bAngularSpring=true
    bGravityOverrideInSimSpace=true
    bOverrideAngularDamping=true
    gravityScale=0.35
    simulationSpace=RootRelative
    alphaBoolBlend {blendOption=Linear}
    boundBone {boneName=dyn_skirt_flap_fb_l}
    physicsBodyDefinitions [{
      sphereCollisionRadius=0
      boundBone {boneName=dyn_skirt_flap_fb_l}
      boxExtents=<10 20 20>
      constraintSetup {angularLimitsMax=<0 2.5 20>, angularLimitsMin=<0 -2.5 0>, angularTarget=<0 0 0>}
      localJointOffset=<10 0 0>
     }]
   }
   n106 = Anim Dynamics - Bone: dyn_skirt_flap_ga_l(ComponentPose=@n105, Alpha=1, ExternalForce="0,0,0") {
    angularDampingOverride=1
    angularSpringConstant=1500
    bAngularSpring=true
    bGravityOverrideInSimSpace=true
    bOverrideAngularDamping=true
    gravityScale=0.35
    simulationSpace=RootRelative
    alphaBoolBlend {blendOption=Linear}
    boundBone {boneName=dyn_skirt_flap_ga_l}
    physicsBodyDefinitions [{
      sphereCollisionRadius=0
      boundBone {boneName=dyn_skirt_flap_ga_l}
      boxExtents=<10 20 20>
      constraintSetup {angularLimitsMax=<0 2.5 20>, angularLimitsMin=<0 -2.5 0>, angularTarget=<0 0 0>}
      localJointOffset=<10 0 0>
     }]
   }
   n107 = Anim Dynamics - Bone: dyn_skirt_flap_gb_l(ComponentPose=@n106, Alpha=1, ExternalForce="0,0,0") {
    angularDampingOverride=1
    angularSpringConstant=1500
    bAngularSpring=true
    bGravityOverrideInSimSpace=true
    bOverrideAngularDamping=true
    gravityScale=0.35
    simulationSpace=RootRelative
    alphaBoolBlend {blendOption=Linear}
    boundBone {boneName=dyn_skirt_flap_gb_l}
    physicsBodyDefinitions [{
      sphereCollisionRadius=0
      boundBone {boneName=dyn_skirt_flap_gb_l}
      boxExtents=<10 20 20>
      constraintSetup {angularLimitsMax=<0 2.5 20>, angularLimitsMin=<0 -2.5 0>, angularTarget=<0 0 0>}
      localJointOffset=<10 0 0>
     }]
   }
   n108 = Anim Dynamics - Bone: dyn_skirt_flap_aa_r(ComponentPose=@n107, Alpha=1, ExternalForce="0,0,0") {
    angularDampingOverride=1
    angularSpringConstant=1500
    bAngularSpring=true
    bGravityOverrideInSimSpace=true
    bOverrideAngularDamping=true
    gravityScale=0.35
    simulationSpace=RootRelative
    alphaBoolBlend {blendOption=Linear}
    boundBone {boneName=dyn_skirt_flap_aa_r}
    physicsBodyDefinitions [{
      sphereCollisionRadius=0
      boundBone {boneName=dyn_skirt_flap_aa_r}
      boxExtents=<10 20 20>
      constraintSetup {angularLimitsMax=<0 2.5 20>, angularLimitsMin=<0 -2.5 0>, angularTarget=<0 0 0>}
      localJointOffset=<-10 0 0>
     }]
   }
   n109 = Anim Dynamics - Bone: dyn_skirt_flap_ab_r(ComponentPose=@n108, Alpha=1, ExternalForce="0,0,0") {
    angularDampingOverride=1
    angularSpringConstant=1500
    bAngularSpring=true
    bGravityOverrideInSimSpace=true
    bOverrideAngularDamping=true
    gravityScale=0.35
    simulationSpace=RootRelative
    alphaBoolBlend {blendOption=Linear}
    boundBone {boneName=dyn_skirt_flap_ab_r}
    physicsBodyDefinitions [{
      sphereCollisionRadius=0
      boundBone {boneName=dyn_skirt_flap_ab_r}
      boxExtents=<10 20 20>
      constraintSetup {angularLimitsMax=<0 2.5 20>, angularLimitsMin=<0 -2.5 0>, angularTarget=<0 0 0>}
      localJointOffset=<-10 0 0>
     }]
   }
   n110 = Anim Dynamics - Bone: dyn_skirt_flap_ba_r(ComponentPose=@n109, Alpha=1, ExternalForce="0,0,0") {
    angularDampingOverride=1
    angularSpringConstant=1500
    bAngularSpring=true
    bGravityOverrideInSimSpace=true
    bOverrideAngularDamping=true
    gravityScale=0.35
    simulationSpace=RootRelative
    alphaBoolBlend {blendOption=Linear}
    boundBone {boneName=dyn_skirt_flap_ba_r}
    physicsBodyDefinitions [{
      sphereCollisionRadius=0
      boundBone {boneName=dyn_skirt_flap_ba_r}
      boxExtents=<10 20 20>
      constraintSetup {angularLimitsMax=<0 2.5 20>, angularLimitsMin=<0 -2.5 0>, angularTarget=<0 0 0>}
      localJointOffset=<-10 0 0>
     }]
   }
   n111 = Anim Dynamics - Bone: dyn_skirt_flap_bb_r(ComponentPose=@n110, Alpha=1, ExternalForce="0,0,0") {
    angularDampingOverride=1
    angularSpringConstant=1500
    bAngularSpring=true
    bGravityOverrideInSimSpace=true
    bOverrideAngularDamping=true
    gravityScale=0.35
    simulationSpace=RootRelative
    alphaBoolBlend {blendOption=Linear}
    boundBone {boneName=dyn_skirt_flap_bb_r}
    physicsBodyDefinitions [{
      sphereCollisionRadius=0
      boundBone {boneName=dyn_skirt_flap_bb_r}
      boxExtents=<10 20 20>
      constraintSetup {angularLimitsMax=<0 2.5 20>, angularLimitsMin=<0 -2.5 0>, angularTarget=<0 0 0>}
      localJointOffset=<-10 0 0>
     }]
   }
   n112 = Anim Dynamics - Bone: dyn_skirt_flap_ca_r(ComponentPose=@n111, Alpha=1, ExternalForce="0,0,0") {
    angularDampingOverride=1
    angularSpringConstant=1500
    bAngularSpring=true
    bGravityOverrideInSimSpace=true
    bOverrideAngularDamping=true
    gravityScale=0.35
    simulationSpace=RootRelative
    alphaBoolBlend {blendOption=Linear}
    boundBone {boneName=dyn_skirt_flap_ca_r}
    physicsBodyDefinitions [{
      sphereCollisionRadius=0
      boundBone {boneName=dyn_skirt_flap_ca_r}
      boxExtents=<10 20 20>
      constraintSetup {angularLimitsMax=<0 2.5 20>, angularLimitsMin=<0 -2.5 0>, angularTarget=<0 0 0>}
      localJointOffset=<-10 0 0>
     }]
   }
   n113 = Anim Dynamics - Bone: dyn_skirt_flap_cb_r(ComponentPose=@n112, Alpha=1, ExternalForce="0,0,0") {
    angularDampingOverride=1
    angularSpringConstant=1500
    bAngularSpring=true
    bGravityOverrideInSimSpace=true
    bOverrideAngularDamping=true
    gravityScale=0.35
    simulationSpace=RootRelative
    alphaBoolBlend {blendOption=Linear}
    boundBone {boneName=dyn_skirt_flap_cb_r}
    physicsBodyDefinitions [{
      sphereCollisionRadius=0
      boundBone {boneName=dyn_skirt_flap_cb_r}
      boxExtents=<10 20 20>
      constraintSetup {angularLimitsMax=<0 2.5 20>, angularLimitsMin=<0 -2.5 0>, angularTarget=<0 0 0>}
      localJointOffset=<-10 0 0>
     }]
   }
   n114 = Anim Dynamics - Bone: dyn_skirt_flap_da_r(ComponentPose=@n113, Alpha=1, ExternalForce="0,0,0") {
    angularDampingOverride=1
    angularSpringConstant=1500
    bAngularSpring=true
    bGravityOverrideInSimSpace=true
    bOverrideAngularDamping=true
    gravityScale=0.35
    simulationSpace=RootRelative
    alphaBoolBlend {blendOption=Linear}
    boundBone {boneName=dyn_skirt_flap_da_r}
    physicsBodyDefinitions [{
      sphereCollisionRadius=0
      boundBone {boneName=dyn_skirt_flap_da_r}
      boxExtents=<10 20 20>
      constraintSetup {angularLimitsMax=<0 2.5 20>, angularLimitsMin=<0 -2.5 0>, angularTarget=<0 0 0>}
      localJointOffset=<-10 0 0>
     }]
   }
   n115 = Anim Dynamics - Bone: dyn_skirt_flap_db_r(ComponentPose=@n114, Alpha=1, ExternalForce="0,0,0") {
    angularDampingOverride=1
    angularSpringConstant=1500
    bAngularSpring=true
    bGravityOverrideInSimSpace=true
    bOverrideAngularDamping=true
    gravityScale=0.35
    simulationSpace=RootRelative
    alphaBoolBlend {blendOption=Linear}
    boundBone {boneName=dyn_skirt_flap_db_r}
    physicsBodyDefinitions [{
      sphereCollisionRadius=0
      boundBone {boneName=dyn_skirt_flap_db_r}
      boxExtents=<10 20 20>
      constraintSetup {angularLimitsMax=<0 2.5 20>, angularLimitsMin=<0 -2.5 0>, angularTarget=<0 0 0>}
      localJointOffset=<-10 0 0>
     }]
   }
   n116 = Anim Dynamics - Bone: dyn_skirt_flap_ea_r(ComponentPose=@n115, Alpha=1, ExternalForce="0,0,0") {
    angularDampingOverride=1
    angularSpringConstant=1500
    bAngularSpring=true
    bGravityOverrideInSimSpace=true
    bOverrideAngularDamping=true
    gravityScale=0.35
    simulationSpace=RootRelative
    alphaBoolBlend {blendOption=Linear}
    boundBone {boneName=dyn_skirt_flap_ea_r}
    physicsBodyDefinitions [{
      sphereCollisionRadius=0
      boundBone {boneName=dyn_skirt_flap_ea_r}
      boxExtents=<10 20 20>
      constraintSetup {angularLimitsMax=<0 2.5 20>, angularLimitsMin=<0 -2.5 0>, angularTarget=<0 0 0>}
      localJointOffset=<-10 0 0>
     }]
   }
   n117 = Anim Dynamics - Bone: dyn_skirt_flap_eb_r(ComponentPose=@n116, Alpha=1, ExternalForce="0,0,0") {
    angularDampingOverride=1
    angularSpringConstant=1500
    bAngularSpring=true
    bGravityOverrideInSimSpace=true
    bOverrideAngularDamping=true
    gravityScale=0.35
    simulationSpace=RootRelative
    alphaBoolBlend {blendOption=Linear}
    boundBone {boneName=dyn_skirt_flap_eb_r}
    physicsBodyDefinitions [{
      sphereCollisionRadius=0
      boundBone {boneName=dyn_skirt_flap_eb_r}
      boxExtents=<10 20 20>
      constraintSetup {angularLimitsMax=<0 2.5 20>, angularLimitsMin=<0 -2.5 0>, angularTarget=<0 0 0>}
      localJointOffset=<-10 0 0>
     }]
   }
   n118 = Anim Dynamics - Bone: dyn_skirt_flap_fa_r(ComponentPose=@n117, Alpha=1, ExternalForce="0,0,0") {
    angularDampingOverride=1
    angularSpringConstant=1500
    bAngularSpring=true
    bGravityOverrideInSimSpace=true
    bOverrideAngularDamping=true
    gravityScale=0.35
    simulationSpace=RootRelative
    alphaBoolBlend {blendOption=Linear}
    boundBone {boneName=dyn_skirt_flap_fa_r}
    physicsBodyDefinitions [{
      sphereCollisionRadius=0
      boundBone {boneName=dyn_skirt_flap_fa_r}
      boxExtents=<10 20 20>
      constraintSetup {angularLimitsMax=<0 2.5 20>, angularLimitsMin=<0 -2.5 0>, angularTarget=<0 0 0>}
      localJointOffset=<-10 0 0>
     }]
   }
   n119 = Anim Dynamics - Bone: dyn_skirt_flap_fb_r(ComponentPose=@n118, Alpha=1, ExternalForce="0,0,0") {
    angularDampingOverride=1
    angularSpringConstant=1500
    bAngularSpring=true
    bGravityOverrideInSimSpace=true
    bOverrideAngularDamping=true
    gravityScale=0.35
    simulationSpace=RootRelative
    alphaBoolBlend {blendOption=Linear}
    boundBone {boneName=dyn_skirt_flap_fb_r}
    physicsBodyDefinitions [{
      sphereCollisionRadius=0
      boundBone {boneName=dyn_skirt_flap_fb_r}
      boxExtents=<10 20 20>
      constraintSetup {angularLimitsMax=<0 2.5 20>, angularLimitsMin=<0 -2.5 0>, angularTarget=<0 0 0>}
      localJointOffset=<-10 0 0>
     }]
   }
   n120 = Anim Dynamics - Bone: dyn_skirt_flap_ga_r(ComponentPose=@n119, Alpha=1, ExternalForce="0,0,0") {
    angularDampingOverride=1
    angularSpringConstant=1500
    bAngularSpring=true
    bGravityOverrideInSimSpace=true
    bOverrideAngularDamping=true
    gravityScale=0.35
    simulationSpace=RootRelative
    alphaBoolBlend {blendOption=Linear}
    boundBone {boneName=dyn_skirt_flap_ga_r}
    physicsBodyDefinitions [{
      sphereCollisionRadius=0
      boundBone {boneName=dyn_skirt_flap_ga_r}
      boxExtents=<10 20 20>
      constraintSetup {angularLimitsMax=<0 2.5 20>, angularLimitsMin=<0 -2.5 0>, angularTarget=<0 0 0>}
      localJointOffset=<-10 0 0>
     }]
   }
   n121 = Anim Dynamics - Bone: dyn_skirt_flap_gb_r(ComponentPose=@n120, Alpha=1, ExternalForce="0,0,0") {
    angularDampingOverride=1
    angularSpringConstant=1500
    bAngularSpring=true
    bGravityOverrideInSimSpace=true
    bOverrideAngularDamping=true
    gravityScale=0.35
    simulationSpace=RootRelative
    alphaBoolBlend {blendOption=Linear}
    boundBone {boneName=dyn_skirt_flap_gb_r}
    physicsBodyDefinitions [{
      sphereCollisionRadius=0
      boundBone {boneName=dyn_skirt_flap_gb_r}
      boxExtents=<10 20 20>
      constraintSetup {angularLimitsMax=<0 2.5 20>, angularLimitsMin=<0 -2.5 0>, angularTarget=<0 0 0>}
      localJointOffset=<-10 0 0>
     }]
   }
   Save cached pose 'SkirtFlaps_Pos'(Pose=Component To Local(ComponentPose=@n121)) {cacheName=SkirtFlaps_Pos}
   n122 = "Transform (Modify) Bone - Bone: dyn_parent_shoulderpad_r"(
    ComponentPose=Local To Component(LocalPose=Use cached pose 'Base_Pose'())
    Alpha=Get arm_fwd_r() -> (arm_fwd_r:double)
    Translation="0.5,2,0"
    Rotation="-10,5,48"
    Scale="1,0.9,0.9"
   ) {
    rotationMode=BMM_Additive
    rotationSpace=BCS_BoneSpace
    scaleMode=BMM_Additive
    scaleSpace=BCS_BoneSpace
    translationMode=BMM_Additive
    translationSpace=BCS_BoneSpace
    alphaBoolBlend {blendOption=Linear}
    alphaScaleBiasClamp {interpSpeedDecreasing=25, interpSpeedIncreasing=25, inRange {max=85}}
    boneToModify {boneName=dyn_parent_shoulderpad_r}
    rotation=<rot -10 5 48>
    scale {y=0.9, z=0.9}
    translation=<0.5 2>
   }
   n123 = "Transform (Modify) Bone - Bone: dyn_parent_shoulderpad_r"(
    ComponentPose=@n122
    Alpha=Get clavicle_up_r() -> (clavicle_up_r:double)
    Translation="0,-1,-7"
    Rotation="-25,0.000,0.000"
    Scale="0.9,0.9,0.9"
   ) {
    rotationMode=BMM_Additive
    rotationSpace=BCS_BoneSpace
    scaleMode=BMM_Additive
    scaleSpace=BCS_BoneSpace
    translationMode=BMM_Additive
    translationSpace=BCS_BoneSpace
    alphaBoolBlend {blendOption=Linear}
    alphaScaleBiasClamp {interpSpeedDecreasing=25, interpSpeedIncreasing=25, inRange {max=85}}
    boneToModify {boneName=dyn_parent_shoulderpad_r}
    rotation {pitch=-25}
    scale=<0.9 0.9 0.9>
    translation {y=-1, z=-7}
   }
   n124 = Get skydive_alpha() -> (skydive_alpha:double)
   n125 = "Constraint - dyn_parent_shoulderpad_r (P)"(ComponentPose=@n123, Alpha=@n124, ConstraintWeights_0=1) {
    alphaBoolBlend {blendOption=Linear}
    boneToModify {boneName=dyn_parent_shoulderpad_r}
    constraintSetup [{targetBone {boneName=upperarm_r}}]
    constraintWeights [1]
   }
   n126 = "Transform (Modify) Bone - Bone: dyn_parent_shoulderpad_r"(ComponentPose=@n125, Alpha=@n124, Translation="5,3,-4", Rotation="-10,15,15", Scale="0.9,0.9,0.9") {
    rotationMode=BMM_Additive
    rotationSpace=BCS_BoneSpace
    scaleMode=BMM_Additive
    scaleSpace=BCS_BoneSpace
    translationMode=BMM_Additive
    translationSpace=BCS_BoneSpace
    alphaBoolBlend {blendOption=Linear}
    alphaScaleBiasClamp {interpSpeedDecreasing=25, interpSpeedIncreasing=25, inRange {max=85}}
    boneToModify {boneName=dyn_parent_shoulderpad_r}
    rotation=<rot -10 15 15>
    scale=<0.9 0.9 0.9>
    translation=<5 3 -4>
   }
   n127 = "Constraint - dyn_parent_shoulderpad_l (P)"(ComponentPose=@n126, Alpha=@n124, ConstraintWeights_0=1) {
    alphaBoolBlend {blendOption=Linear}
    boneToModify {boneName=dyn_parent_shoulderpad_l}
    constraintSetup [{targetBone {boneName=upperarm_l}}]
    constraintWeights [1]
   }
   n128 = "Transform (Modify) Bone - Bone: dyn_parent_shoulderpad_l"(ComponentPose=@n127, Alpha=@n124, Translation="5,-3,-4", Rotation="-10,-15,-15", Scale="0.9,0.9,0.9") {
    rotationMode=BMM_Additive
    rotationSpace=BCS_BoneSpace
    scaleMode=BMM_Additive
    scaleSpace=BCS_BoneSpace
    translationMode=BMM_Additive
    translationSpace=BCS_BoneSpace
    alphaBoolBlend {blendOption=Linear}
    alphaScaleBiasClamp {interpSpeedDecreasing=25, interpSpeedIncreasing=25, inRange {max=85}}
    boneToModify {boneName=dyn_parent_shoulderpad_l}
    rotation=<rot -10 -15 -15>
    scale=<0.9 0.9 0.9>
    translation=<5 -3 -4>
   }
   n129 = Anim Dynamics - Bone: dyn_shldr_feather_a_l(ComponentPose=@n128, Alpha=0.3, ExternalForce="0,0,0") {
    angularDampingOverride=1
    angularSpringConstant=1500
    bAngularSpring=true
    bGravityOverrideInSimSpace=true
    bOverrideAngularDamping=true
    gravityScale=0.35
    simulationSpace=RootRelative
    alphaBoolBlend {blendOption=Linear}
    boundBone {boneName=dyn_shldr_feather_a_l}
    physicsBodyDefinitions [{
      sphereCollisionRadius=0
      boundBone {boneName=dyn_shldr_feather_a_l}
      boxExtents=<10 20 20>
      constraintSetup {angularLimitsMax=<0 5 0>, angularLimitsMin=<0 -15 0>, angularTarget=<0 0 0>}
      localJointOffset=<10 0 0>
     }]
   }
   n130 = Anim Dynamics - Bone: dyn_shldr_feather_b_l(ComponentPose=@n129, Alpha=0.4, ExternalForce="0,0,0") {
    angularDampingOverride=1
    angularSpringConstant=1500
    bAngularSpring=true
    bGravityOverrideInSimSpace=true
    bOverrideAngularDamping=true
    gravityScale=0.35
    simulationSpace=RootRelative
    alphaBoolBlend {blendOption=Linear}
    boundBone {boneName=dyn_shldr_feather_b_l}
    physicsBodyDefinitions [{
      sphereCollisionRadius=0
      boundBone {boneName=dyn_shldr_feather_b_l}
      boxExtents=<10 20 20>
      constraintSetup {angularLimitsMax=<0 5 0>, angularLimitsMin=<0 -15 0>, angularTarget=<0 0 0>}
      localJointOffset=<10 0 0>
     }]
   }
   n131 = Anim Dynamics - Bone: dyn_shldr_feather_c_l(ComponentPose=@n130, Alpha=0.5, ExternalForce="0,0,0") {
    angularDampingOverride=1
    angularSpringConstant=1500
    bAngularSpring=true
    bGravityOverrideInSimSpace=true
    bOverrideAngularDamping=true
    gravityScale=0.35
    simulationSpace=RootRelative
    alphaBoolBlend {blendOption=Linear}
    boundBone {boneName=dyn_shldr_feather_c_l}
    physicsBodyDefinitions [{
      sphereCollisionRadius=0
      boundBone {boneName=dyn_shldr_feather_c_l}
      boxExtents=<10 20 20>
      constraintSetup {angularLimitsMax=<0 5 0>, angularLimitsMin=<0 -15 0>, angularTarget=<0 0 0>}
      localJointOffset=<10 0 0>
     }]
   }
   n132 = Anim Dynamics - Bone: dyn_shldr_feather_d_l(ComponentPose=@n131, Alpha=0.5, ExternalForce="0,0,0") {
    angularDampingOverride=1
    angularSpringConstant=1500
    bAngularSpring=true
    bGravityOverrideInSimSpace=true
    bOverrideAngularDamping=true
    gravityScale=0.35
    simulationSpace=RootRelative
    alphaBoolBlend {blendOption=Linear}
    boundBone {boneName=dyn_shldr_feather_d_l}
    physicsBodyDefinitions [{
      sphereCollisionRadius=0
      boundBone {boneName=dyn_shldr_feather_d_l}
      boxExtents=<10 20 20>
      constraintSetup {angularLimitsMax=<0 5 0>, angularLimitsMin=<0 -15 0>, angularTarget=<0 0 0>}
      localJointOffset=<10 0 0>
     }]
   }
   n133 = Anim Dynamics - Bone: dyn_shldr_feather_e_l(ComponentPose=@n132, Alpha=0.5, ExternalForce="0,0,0") {
    angularDampingOverride=1
    angularSpringConstant=1500
    bAngularSpring=true
    bGravityOverrideInSimSpace=true
    bOverrideAngularDamping=true
    gravityScale=0.35
    simulationSpace=RootRelative
    alphaBoolBlend {blendOption=Linear}
    boundBone {boneName=dyn_shldr_feather_e_l}
    physicsBodyDefinitions [{
      sphereCollisionRadius=0
      boundBone {boneName=dyn_shldr_feather_e_l}
      boxExtents=<10 20 20>
      constraintSetup {angularLimitsMax=<0 5 0>, angularLimitsMin=<0 -15 0>, angularTarget=<0 0 0>}
      localJointOffset=<10 0 0>
     }]
   }
   n134 = Anim Dynamics - Bone: dyn_shldr_feather_f_l(ComponentPose=@n133, Alpha=0.5, ExternalForce="0,0,0") {
    angularDampingOverride=1
    angularSpringConstant=1500
    bAngularSpring=true
    bGravityOverrideInSimSpace=true
    bOverrideAngularDamping=true
    gravityScale=0.35
    simulationSpace=RootRelative
    alphaBoolBlend {blendOption=Linear}
    boundBone {boneName=dyn_shldr_feather_f_l}
    physicsBodyDefinitions [{
      sphereCollisionRadius=0
      boundBone {boneName=dyn_shldr_feather_f_l}
      boxExtents=<10 20 20>
      constraintSetup {angularLimitsMax=<0 5 0>, angularLimitsMin=<0 -15 0>, angularTarget=<0 0 0>}
      localJointOffset=<10 0 0>
     }]
   }
   n135 = Anim Dynamics - Bone: dyn_shldr_feather_g_l(ComponentPose=@n134, Alpha=0.5, ExternalForce="0,0,0") {
    angularDampingOverride=1
    angularSpringConstant=1500
    bAngularSpring=true
    bGravityOverrideInSimSpace=true
    bOverrideAngularDamping=true
    gravityScale=0.35
    simulationSpace=RootRelative
    alphaBoolBlend {blendOption=Linear}
    boundBone {boneName=dyn_shldr_feather_g_l}
    physicsBodyDefinitions [{
      sphereCollisionRadius=0
      boundBone {boneName=dyn_shldr_feather_g_l}
      boxExtents=<10 20 20>
      constraintSetup {angularLimitsMax=<0 5 0>, angularLimitsMin=<0 -15 0>, angularTarget=<0 0 0>}
      localJointOffset=<10 0 0>
     }]
   }
   n136 = Anim Dynamics - Bone: dyn_shldr_feather_h_l(ComponentPose=@n135, Alpha=0.4, ExternalForce="0,0,0") {
    angularDampingOverride=1
    angularSpringConstant=1500
    bAngularSpring=true
    bGravityOverrideInSimSpace=true
    bOverrideAngularDamping=true
    gravityScale=0.35
    simulationSpace=RootRelative
    alphaBoolBlend {blendOption=Linear}
    boundBone {boneName=dyn_shldr_feather_h_l}
    physicsBodyDefinitions [{
      sphereCollisionRadius=0
      boundBone {boneName=dyn_shldr_feather_h_l}
      boxExtents=<10 20 20>
      constraintSetup {angularLimitsMax=<0 5 0>, angularLimitsMin=<0 -15 0>, angularTarget=<0 0 0>}
      localJointOffset=<10 0 0>
     }]
   }
   n137 = Anim Dynamics - Bone: dyn_shldr_feather_i_l(ComponentPose=@n136, Alpha=0.3, ExternalForce="0,0,0") {
    angularDampingOverride=1
    angularSpringConstant=1500
    bAngularSpring=true
    bGravityOverrideInSimSpace=true
    bOverrideAngularDamping=true
    gravityScale=0.35
    simulationSpace=RootRelative
    alphaBoolBlend {blendOption=Linear}
    boundBone {boneName=dyn_shldr_feather_i_l}
    physicsBodyDefinitions [{
      sphereCollisionRadius=0
      boundBone {boneName=dyn_shldr_feather_i_l}
      boxExtents=<10 20 20>
      constraintSetup {angularLimitsMax=<0 5 0>, angularLimitsMin=<0 -15 0>, angularTarget=<0 0 0>}
      localJointOffset=<10 0 0>
     }]
   }
   n138 = Anim Dynamics - Bone: dyn_shldr_feather_a_r(ComponentPose=@n137, Alpha=0.3, ExternalForce="0,0,0") {
    angularDampingOverride=1
    angularSpringConstant=1500
    bAngularSpring=true
    bGravityOverrideInSimSpace=true
    bOverrideAngularDamping=true
    gravityScale=0.35
    simulationSpace=RootRelative
    alphaBoolBlend {blendOption=Linear}
    boundBone {boneName=dyn_shldr_feather_a_r}
    physicsBodyDefinitions [{
      sphereCollisionRadius=0
      boundBone {boneName=dyn_shldr_feather_a_r}
      boxExtents=<10 20 20>
      constraintSetup {angularLimitsMax=<0 5 0>, angularLimitsMin=<0 -15 0>, angularTarget=<0 0 0>}
      localJointOffset=<-10 0 0>
     }]
   }
   n139 = Anim Dynamics - Bone: dyn_shldr_feather_b_r(ComponentPose=@n138, Alpha=0.4, ExternalForce="0,0,0") {
    angularDampingOverride=1
    angularSpringConstant=1500
    bAngularSpring=true
    bGravityOverrideInSimSpace=true
    bOverrideAngularDamping=true
    gravityScale=0.35
    simulationSpace=RootRelative
    alphaBoolBlend {blendOption=Linear}
    boundBone {boneName=dyn_shldr_feather_b_r}
    physicsBodyDefinitions [{
      sphereCollisionRadius=0
      boundBone {boneName=dyn_shldr_feather_b_r}
      boxExtents=<10 20 20>
      constraintSetup {angularLimitsMax=<0 5 0>, angularLimitsMin=<0 -15 0>, angularTarget=<0 0 0>}
      localJointOffset=<-10 0 0>
     }]
   }
   n140 = Anim Dynamics - Bone: dyn_shldr_feather_c_r(ComponentPose=@n139, Alpha=0.5, ExternalForce="0,0,0") {
    angularDampingOverride=1
    angularSpringConstant=1500
    bAngularSpring=true
    bGravityOverrideInSimSpace=true
    bOverrideAngularDamping=true
    gravityScale=0.35
    simulationSpace=RootRelative
    alphaBoolBlend {blendOption=Linear}
    boundBone {boneName=dyn_shldr_feather_c_r}
    physicsBodyDefinitions [{
      sphereCollisionRadius=0
      boundBone {boneName=dyn_shldr_feather_c_r}
      boxExtents=<10 20 20>
      constraintSetup {angularLimitsMax=<0 5 0>, angularLimitsMin=<0 -15 0>, angularTarget=<0 0 0>}
      localJointOffset=<-10 0 0>
     }]
   }
   n141 = Anim Dynamics - Bone: dyn_shldr_feather_d_r(ComponentPose=@n140, Alpha=0.5, ExternalForce="0,0,0") {
    angularDampingOverride=1
    angularSpringConstant=1500
    bAngularSpring=true
    bGravityOverrideInSimSpace=true
    bOverrideAngularDamping=true
    gravityScale=0.35
    simulationSpace=RootRelative
    alphaBoolBlend {blendOption=Linear}
    boundBone {boneName=dyn_shldr_feather_d_r}
    physicsBodyDefinitions [{
      sphereCollisionRadius=0
      boundBone {boneName=dyn_shldr_feather_d_r}
      boxExtents=<10 20 20>
      constraintSetup {angularLimitsMax=<0 5 0>, angularLimitsMin=<0 -15 0>, angularTarget=<0 0 0>}
      localJointOffset=<-10 0 0>
     }]
   }
   n142 = Anim Dynamics - Bone: dyn_shldr_feather_e_r(ComponentPose=@n141, Alpha=0.5, ExternalForce="0,0,0") {
    angularDampingOverride=1
    angularSpringConstant=1500
    bAngularSpring=true
    bGravityOverrideInSimSpace=true
    bOverrideAngularDamping=true
    gravityScale=0.35
    simulationSpace=RootRelative
    alphaBoolBlend {blendOption=Linear}
    boundBone {boneName=dyn_shldr_feather_e_r}
    physicsBodyDefinitions [{
      sphereCollisionRadius=0
      boundBone {boneName=dyn_shldr_feather_e_r}
      boxExtents=<10 20 20>
      constraintSetup {angularLimitsMax=<0 5 0>, angularLimitsMin=<0 -15 0>, angularTarget=<0 0 0>}
      localJointOffset=<-10 0 0>
     }]
   }
   n143 = Anim Dynamics - Bone: dyn_shldr_feather_f_r(ComponentPose=@n142, Alpha=0.5, ExternalForce="0,0,0") {
    angularDampingOverride=1
    angularSpringConstant=1500
    bAngularSpring=true
    bGravityOverrideInSimSpace=true
    bOverrideAngularDamping=true
    gravityScale=0.35
    simulationSpace=RootRelative
    alphaBoolBlend {blendOption=Linear}
    boundBone {boneName=dyn_shldr_feather_f_r}
    physicsBodyDefinitions [{
      sphereCollisionRadius=0
      boundBone {boneName=dyn_shldr_feather_f_r}
      boxExtents=<10 20 20>
      constraintSetup {angularLimitsMax=<0 5 0>, angularLimitsMin=<0 -15 0>, angularTarget=<0 0 0>}
      localJointOffset=<-10 0 0>
     }]
   }
   n144 = Anim Dynamics - Bone: dyn_shldr_feather_g_r(ComponentPose=@n143, Alpha=0.5, ExternalForce="0,0,0") {
    angularDampingOverride=1
    angularSpringConstant=1500
    bAngularSpring=true
    bGravityOverrideInSimSpace=true
    bOverrideAngularDamping=true
    gravityScale=0.35
    simulationSpace=RootRelative
    alphaBoolBlend {blendOption=Linear}
    boundBone {boneName=dyn_shldr_feather_g_r}
    physicsBodyDefinitions [{
      sphereCollisionRadius=0
      boundBone {boneName=dyn_shldr_feather_g_r}
      boxExtents=<10 20 20>
      constraintSetup {angularLimitsMax=<0 5 0>, angularLimitsMin=<0 -15 0>, angularTarget=<0 0 0>}
      localJointOffset=<-10 0 0>
     }]
   }
   n145 = Anim Dynamics - Bone: dyn_shldr_feather_h_r(ComponentPose=@n144, Alpha=0.4, ExternalForce="0,0,0") {
    angularDampingOverride=1
    angularSpringConstant=1500
    bAngularSpring=true
    bGravityOverrideInSimSpace=true
    bOverrideAngularDamping=true
    gravityScale=0.35
    simulationSpace=RootRelative
    alphaBoolBlend {blendOption=Linear}
    boundBone {boneName=dyn_shldr_feather_h_r}
    physicsBodyDefinitions [{
      sphereCollisionRadius=0
      boundBone {boneName=dyn_shldr_feather_h_r}
      boxExtents=<10 20 20>
      constraintSetup {angularLimitsMax=<0 5 0>, angularLimitsMin=<0 -15 0>, angularTarget=<0 0 0>}
      localJointOffset=<-10 0 0>
     }]
   }
   n146 = Anim Dynamics - Bone: dyn_shldr_feather_i_r(ComponentPose=@n145, Alpha=0.3, ExternalForce="0,0,0") {
    angularDampingOverride=1
    angularSpringConstant=1500
    bAngularSpring=true
    bGravityOverrideInSimSpace=true
    bOverrideAngularDamping=true
    gravityScale=0.35
    simulationSpace=RootRelative
    alphaBoolBlend {blendOption=Linear}
    boundBone {boneName=dyn_shldr_feather_i_r}
    physicsBodyDefinitions [{
      sphereCollisionRadius=0
      boundBone {boneName=dyn_shldr_feather_i_r}
      boxExtents=<10 20 20>
      constraintSetup {angularLimitsMax=<0 5 0>, angularLimitsMin=<0 -15 0>, angularTarget=<0 0 0>}
      localJointOffset=<-10 0 0>
     }]
   }
   Save cached pose 'ShoulderPads_Pose'(Pose=Component To Local(ComponentPose=@n146)) {cacheName=ShoulderPads_Pose}
   n147 = Control Rig(Source=Copy Pose From Mesh() {bUseAttachedParent=true}, Alpha=Get skydive_alpha_neg() -> (skydive_alpha_neg:double)) {
    alphaBoolBlend {blendOption=Linear}
    controlRigAssetReference {blueprintRigClass=&/M_MED_Banana_Leader_CtrlRig.M_MED_Banana_Leader_CtrlRig_C}
   }
   n148 = "Constraint - dyn_volume_base_lowerarm_l (PP)"!(ComponentPose=Local To Component(LocalPose=@n147), Alpha=1, ConstraintWeights_0=0.65, ConstraintWeights_1:float=0.35) -> (Pose) {
    alphaBoolBlend {blendOption=Linear}
    boneToModify {boneName=dyn_volume_base_lowerarm_l}
    constraintSetup [{targetBone {boneName=upperarm_l}}, {targetBone {boneName=lowerarm_l}}]
    constraintWeights [1,1]
   }
   n149 = "Constraint - dyn_volume_base_lowerarm_r (PP)"!(ComponentPose=@n148, Alpha=1, ConstraintWeights_0=0.65, ConstraintWeights_1:float=0.35) -> (Pose) {
    alphaBoolBlend {blendOption=Linear}
    boneToModify {boneName=dyn_volume_base_lowerarm_r}
    constraintSetup [{targetBone {boneName=upperarm_r}}, {targetBone {boneName=lowerarm_r}}]
    constraintWeights [1,1]
   }
   n150 = "Transform (Modify) Bone - Bone: dyn_volume_lowerarm_l"(ComponentPose=@n149, Alpha=Get elbow_rot_l() -> (elbow_rot_l:double), Translation="0,-2,0", Rotation="0,0,0", Scale="0.5,1,1") {
    rotationMode=BMM_Additive
    rotationSpace=BCS_BoneSpace
    scaleMode=BMM_Additive
    scaleSpace=BCS_BoneSpace
    translationMode=BMM_Additive
    translationSpace=BCS_BoneSpace
    alphaBoolBlend {blendOption=Linear}
    alphaScaleBiasClamp {bMapRange=true, inRange {max=80}}
    boneToModify {boneName=dyn_volume_lowerarm_l}
    scale {x=0.5}
    translation {y=-2}
   }
   n151 = "Transform (Modify) Bone - Bone: dyn_volume_lowerarm_r"(ComponentPose=@n150, Alpha=Get elbow_rot_r() -> (elbow_rot_r:double), Translation="0,2,0", Rotation="0,0,0", Scale="0.5,1,1") {
    rotationMode=BMM_Additive
    rotationSpace=BCS_BoneSpace
    scaleMode=BMM_Additive
    scaleSpace=BCS_BoneSpace
    translationMode=BMM_Additive
    translationSpace=BCS_BoneSpace
    alphaBoolBlend {blendOption=Linear}
    alphaScaleBiasClamp {bMapRange=true, inRange {max=80}}
    boneToModify {boneName=dyn_volume_lowerarm_r}
    scale {x=0.5}
    translation {y=2}
   }
   Save cached pose 'Base_Pose'(Pose=Component To Local(ComponentPose=@n151)) {cacheName=Base_Pose}
  ]}
 EventGraph {nodes [
   Event BlueprintUpdateAnimation()
   > n1 = Sequence()
   > n2 = Sequence!(execute) -> (then_0, then_1, then_2:Exec, then_3:Exec, then_4:Exec, then_5:Exec, then_6:Exec, then_7:Exec)
   n3 = GetDeltaTransformFromRefPose(self=GetOwnerSkeletalMeshComponent(), BoneName=lowerarm_l, BaseName=upperarm_l)
   > Set elbow_rot_l(elbow_rot_l:double=BreakRotator(InRot=BreakTransform(InTransform=@n3).Rotation).Yaw) -> (Output_Get:double)
   n4 = GetDeltaTransformFromRefPose(self=GetOwnerSkeletalMeshComponent(), BoneName=lowerarm_r, BaseName=upperarm_r)
   > Set elbow_rot_r(elbow_rot_r:double=BreakRotator(InRot=BreakTransform(InTransform=@n4).Rotation).Yaw) -> (Output_Get:double)
   n5 = GetOwnerSkeletalMeshComponent()
   n6 = BreakRotator(InRot=BreakTransform(InTransform=GetDeltaTransformFromRefPose(self=@n5, BoneName=clavicle_r, BaseName=spine_05)).Rotation)
   Set r_shldr_up(execute=@n2.then_2, r_shldr_up:double=@n6.Pitch) -> (Output_Get:double)
   > Set r_shldr_fwd(r_shldr_fwd:double=@n6.Yaw) -> (Output_Get:double)
   n7 = GetOwnerSkeletalMeshComponent()
   n8 = BreakRotator(InRot=BreakTransform(InTransform=GetDeltaTransformFromRefPose(self=@n7, BoneName=clavicle_l, BaseName=spine_05)).Rotation)
   > Set l_shldr_up(l_shldr_up:double=@n8.Pitch) -> (Output_Get:double)
   > Set l_shldr_fwd(l_shldr_fwd:double=@n8.Yaw) -> (Output_Get:double)
   n9 = GetDeltaTransformFromRefPose(self=GetOwnerSkeletalMeshComponent(), BoneName=thigh_l, BaseName=pelvis)
   Set thigh_rot_l(execute=@n2.then_1, thigh_rot_l=BreakRotator(InRot=BreakTransform(InTransform=@n9).Rotation).Yaw)
   n10 = GetDeltaTransformFromRefPose(self=GetOwnerSkeletalMeshComponent(), BoneName=thigh_r, BaseName=pelvis)
   > Set thigh_rot_r(thigh_rot_r=BreakRotator(InRot=BreakTransform(InTransform=@n10).Rotation).Yaw)
   n11 = GetOwnerSkeletalMeshComponent()
   n12 = BreakRotator(InRot=BreakTransform(InTransform=GetDeltaTransformFromRefPose(self=@n11, BoneName=upperarm_l, BaseName=clavicle_l)).Rotation)
   Set l_arm_fwd_new(execute=@n2.then_3, l_arm_fwd_new:double=@n12.Yaw) -> (Output_Get:double)
   > Set l_arm_dn_new(l_arm_dn_new:double=@n12.Pitch) -> (Output_Get:double)
   n13 = GetOwnerSkeletalMeshComponent()
   > Set arm_fwd_r(arm_fwd_r:double=GetAlphaFromReferenceTransformDelta(
     SkeletalMeshComponent=@n13
     BoneName=upperarm_r
     BaseName=clavicle_r
     DeltaComponentType=QuaternionTwist
     InputRange=Get Input Range() -> (Input Range:Vector 2D Structure)
     TwistAxis="0,0,1"
    )) -> (Output_Get:double)
   > Set clavicle_up_r(clavicle_up_r:double=GetAlphaFromReferenceTransformDelta(SkeletalMeshComponent=@n13, BoneName=clavicle_r, BaseName=spine_05, DeltaComponentType=EulerY, InputRange=(X=0,Y=50), TwistAxis="0,0,0")) -> (Output_Get:double)
   > Branch(Condition=Get skydive_active() -> (skydive_active:bool))
   > Set arm_fwd_r(arm_fwd_r=0)
   > Set clavicle_up_r(clavicle_up_r=0)
   n14 = Sequence(execute=@n1.then_1)
   n15 = Get bIsParachuteOpened()
   n16 = Get bIsSkydiveFloating()
   > n17 = Branch(Condition=OR Boolean(A=OR Boolean(A=@n15, B=@n16), B=Get bIsSkyDiving()))
   > Set skydive_alpha_neg(skydive_alpha_neg:double=0) -> (Output_Get:double)
   > Set skydive_active(skydive_active:bool=true) -> (Output_Get:bool)
   > Set skydive_alpha(skydive_alpha:double=1) -> (Output_Get:double)
   Set skydive_alpha_neg(execute=@n17.else, skydive_alpha_neg=1)
   > Set skydive_active(skydive_active=false)
   > Set skydive_alpha(skydive_alpha=0)
   n18 = Branch(execute=@n14.then_1, Condition=Get bIsPlayingEmote())
   > Set Input Range(Input Range:Vector 2D Structure=(X=-20,Y=90)) -> (Output_Get:Vector 2D Structure)
   n19 = Branch(execute=@n18.else, Condition=Get pistol_bool() -> (pistol_bool:bool))
   > Set Input Range(Input Range=(X=-70,Y=-110))
   Set Input Range(execute=@n19.else, Input Range=(X=-35,Y=-105))
   n20 = Sequence(execute=@n14.then_3)
   n21 = Branch(execute=@n20.then_1, Condition=OR Boolean(A=Get bIsJumping(), B=Get bIsFalling()))
   > Set jumpingAlpha(jumpingAlpha:double=1) -> (Output_Get:double)
   Set jumpingAlpha(execute=@n21.else, jumpingAlpha=0)
   n22 = Get bIsMoving2D()
   n23 = Branch(execute=@n20.then_0, Condition=AND Boolean(A=Get bIsCrouching(), B=NOT Boolean(A=@n22)))
   > Set isCrouchingAlpha(isCrouchingAlpha:double=1) -> (Output_Get:double)
   Set isCrouchingAlpha(execute=@n23.else, isCrouchingAlpha=0)
   n24 = Branch(execute=@n20.then_2, Condition=Get bIsPlayingEmote())
   n25 = Cast To FortPlayerAnimInstance(execute=@n24.else, Object=GetAnimInstance(self=GetOwnerSkeletalMeshComponent()))
   > n26 = Sequence!(execute) -> (then_0, then_1)
   n27 = Enum to String!(Enumerator:EFortWeaponCoreAnimation Enum=Get WeaponCoreAnim(self=@n25)) -> (ReturnValue:String)
   > n28 = Branch(Condition=Contains Item(TargetArray:Array of Strings=Get rifleArray() -> (rifleArray:Array of Strings), ItemToFind:"String (by ref)"=@n27))
   > Set isRifle(isRifle:double=1) -> (Output_Get:double)
   > Set pistol_bool(pistol_bool:bool=false) -> (Output_Get:bool)
   n29 = Set isRifle(execute=@n28.else, isRifle=0)
   n30 = Branch(execute=[@n29.then, @n26.then_1], Condition=Contains Item(TargetArray=Get pistolArray() -> (pistolArray:Array of Strings), ItemToFind=@n27))
   > Set isPistol(isPistol:double=1) -> (Output_Get:double)
   > Set pistol_bool(pistol_bool=true)
   Set isPistol(execute=@n30.else, isPistol=1)
   > Set pistol_bool(pistol_bool=false)
   n31 = Branch(Condition=Get bIsRelaxedLevel1() -> (bIsRelaxedLevel1:bool))
   > n32 = Sequence()
   n33 = Enum to String(Enumerator=Get WeaponCoreAnim())
   n34 = Contains Item(TargetArray=Get rifleArray(), ItemToFind=@n33)
   > n35 = Branch(Condition=@n34)
   > Set rifleRelaxedAlpha(rifleRelaxedAlpha:double=1) -> (Output_Get:double)
   Set rifleRelaxedAlpha(execute=@n35.else, rifleRelaxedAlpha=0)
   n36 = Contains Item(TargetArray=Get pistolArray(), ItemToFind=@n33)
   n37 = Branch(execute=@n32.then_1, Condition=@n36)
   > Set pistolRelaxedAlpha(pistolRelaxedAlpha:double=1) -> (Output_Get:double)
   Set pistolRelaxedAlpha(execute=@n37.else, pistolRelaxedAlpha=0)
   n38 = Sequence(execute=@n31.else)
   > Set rifleRelaxedAlpha(rifleRelaxedAlpha=0)
   Set pistolRelaxedAlpha(execute=@n38.then_1, pistolRelaxedAlpha=0)
   Reroute Node()
   n39 = Get bIsRelaxedLevel1()
   > n40 = Branch(Condition=AND Boolean(A=OR Boolean(A=@n34, B=@n36), B=NOT Boolean(A=@n39)))
   > Set weaponAimDownAlpha(weaponAimDownAlpha:double=Get Distance Between Two Sockets!(
     Component:Skeletal Mesh Component Object Reference=GetOwnerSkeletalMeshComponent()
     SocketOrBoneNameA:Name=hand_l
     SocketSpaceA:ERelativeTransformSpace Enum=RTS_World
     SocketOrBoneNameB:Name=pelvis
     SocketSpaceB:ERelativeTransformSpace Enum=RTS_World
     bRemapRange:bool=true
     InRangeMin:float=60
     InRangeMax:float=20
     OutRangeMin:float=0
     OutRangeMax:float=1
    ) -> (ReturnValue:float) {is=title K2Node_CallFunction}) -> (Output_Get:double)
   Set weaponAimDownAlpha(execute=@n40.else, weaponAimDownAlpha=0)
   Reroute Node()
   n41 = GetPartSkeletalMeshComponent()
   > Set l_arm_fwd(l_arm_fwd:double=GetAlphaFromReferenceTransformDelta(
     SkeletalMeshComponent=@n41
     BoneName=lowerarm_l
     BaseName=dyn_arm_measure_l
     DeltaComponentType=TranslationY
     InputRange=(X=5,Y=25)
     TwistAxis="1,0,0"
    )) -> (Output_Get:double)
   n42 = GetPartSkeletalMeshComponent()
   > Set r_arm_fwd(r_arm_fwd:double=GetAlphaFromReferenceTransformDelta(
     SkeletalMeshComponent=@n42
     BoneName=lowerarm_r
     BaseName=dyn_arm_measure_r
     DeltaComponentType=TranslationY
     InputRange=(X=-25,Y=-5)
     OutputRange=(X=1,Y=0)
     TwistAxis="1,0,0"
    )) -> (Output_Get:double)
   Reroute Node()
   > Set l_arm_up(l_arm_up:double=GetAlphaFromReferenceTransformDelta(
     SkeletalMeshComponent=@n41
     BoneName=lowerarm_l
     BaseName=dyn_arm_measure_l
     DeltaComponentType=TranslationZ
     InputRange=(X=10,Y=25)
     TwistAxis="1,0,0"
    )) -> (Output_Get:double)
   > Set r_arm_up(r_arm_up:double=GetAlphaFromReferenceTransformDelta(
     SkeletalMeshComponent=@n42
     BoneName=lowerarm_r
     BaseName=dyn_arm_measure_r
     DeltaComponentType=TranslationZ
     InputRange=(X=-25,Y=-10)
     OutputRange=(X=1,Y=0)
     TwistAxis="1,0,0"
    )) -> (Output_Get:double)
   Reroute Node()
   > Set l_arm_dn(l_arm_dn:double=GetAlphaFromReferenceTransformDelta(
     SkeletalMeshComponent=@n41
     BoneName=lowerarm_l
     BaseName=dyn_arm_measure_l
     DeltaComponentType=TranslationZ
     InputRange=(X=-7,Y=0)
     OutputRange=(X=1,Y=0)
     TwistAxis="1,0,0"
    )) -> (Output_Get:double)
   > Set r_arm_dn(r_arm_dn:double=GetAlphaFromReferenceTransformDelta(
     SkeletalMeshComponent=@n42
     BoneName=lowerarm_r
     BaseName=dyn_arm_measure_r
     DeltaComponentType=TranslationZ
     InputRange=(X=0,Y=7)
     TwistAxis="1,0,0"
    )) -> (Output_Get:double)
   Reroute Node(InputPin:Wildcard) -> (OutputPin:Wildcard)
  ]}
]
variables [
 arm_fwd_r {type=number}
 clavicle_up_r {type=number}
 elbow_rot_l {type=number}
 elbow_rot_r {type=number}
 Input Range {type=Vector2D, default=<-70 -105>}
 isCrouchingAlpha {type=number}
 isPistol {type=number}
 isRifle {type=number}
 jumpingAlpha {type=number}
 l_arm_dn {type=number}
 l_arm_dn_0 {type=number}
 l_arm_dn_new {type=number}
 l_arm_fwd {type=number}
 l_arm_fwd_new {type=number}
 l_arm_up {type=number}
 l_shldr_fwd {type=number}
 l_shldr_up {type=number}
 pistol_bool {type=boolean}
 pistolArray {type="array<string>", default [Pistol, Machine Pistol]}
 pistolRelaxedAlpha {type=number}
 r_arm_dn {type=number}
 r_arm_dn_new {type=number}
 r_arm_fwd {type=number}
 r_arm_fwd_new {type=number}
 r_arm_up {type=number}
 r_shldr_fwd {type=number}
 r_shldr_up {type=number}
 rifleArray {
  type="array<string>"
  default [Shotgun, Rifle, Rocket Laouncher, Grenade Launcher, Assault Rifle, Tactical Shotgun, Sniper Rifle, Shoulder Launcher, crossbow]
 }
 rifleRelaxedAlpha {type=number}
 skydive_active {type=boolean}
 skydive_alpha {type=number}
 skydive_alpha_neg {type=number}
 test {type=number}
 thigh_rot_l {type=number}
 thigh_rot_r {type=number}
 weaponAimDownAlpha {type=number}
]
