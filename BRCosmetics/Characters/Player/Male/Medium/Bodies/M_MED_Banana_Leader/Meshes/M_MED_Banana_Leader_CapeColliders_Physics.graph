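# Physics asset of cape colliders for the medium male Banana Leader body, named as the collision shapes for its cape.
# Each body is a single capsule, covering the pelvis, spine, arms, hands, legs and feet, with no head or neck body.
# Capsules also sit on dynamic shoulder pad, skirt and back flap bones, so those secondary parts count as collision surfaces.
# It has no constraints, so it defines collision volumes only rather than a jointed ragdoll.
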PhysicsAsset /BRCosmetics/Characters/Player/Male/Medium/Bodies/M_MED_Banana_Leader/Meshes/M_MED_Banana_Leader_CapeColliders_Physics
bodies {
 calf_l [{shapeType=Capsule, length=20.9039, radius=7.8853, center=<-19.8529 -2.9883 -0.5457>, rotation=<rot 0 -90.6574 -89.3735>}]
 calf_r [{shapeType=Capsule, length=20.9044, radius=7.886, center=<19.8528 2.9868 0.5479>, rotation=<rot 0 -90.665 -89.3975>}]
 dyn_flap_b_c [{shapeType=Capsule, length=13.3974, radius=5.8234, center=<9.9506 0.1414 0.0009>, rotation=<rot 0 90.5326 -89.9948>}]
 dyn_shoulderpad_l [{shapeType=Capsule, length=2.3242, radius=11.0078, center=<4.5311 -0.6509 4.976>, rotation=<rot 0 -178.7391 -89.1197>}]
 dyn_shoulderpad_r [{shapeType=Capsule, length=2.3242, radius=11.0078, center=<-4.5308 0.6509 -4.9764>, rotation=<rot 0 -178.7391 -89.1197>}]
 dyn_skirt_bk_a_l [{shapeType=Capsule, length=16.098, radius=7.8401, center=<16.4409 -2.757 -0.3368>, rotation=<rot 0 89.3847 -83.2406>}]
 dyn_skirt_bk_a_r [{shapeType=Capsule, length=16.6012, radius=9.0724, center=<-16.7048 2.5309 0.6496>, rotation=<rot 0 88.9921 -76.9716>}]
 dyn_skirt_fr_a_l [{shapeType=Capsule, length=18.5779, radius=9.2235, center=<13.8991 -2.701 -0.4287>, rotation=<rot 0 -91.7266 -84.0574>}]
 dyn_skirt_fr_a_r [{shapeType=Capsule, length=18.1557, radius=7.6051, center=<-13.8018 2.9967 -0.2649>, rotation=<rot 0 -92.1467 -87.902>}]
 dyn_skirt_side_aa_l [{shapeType=Capsule, length=17.1433, radius=5.8129, center=<16.0697 -3.6649 0.4693>, rotation=<rot 0 86.111 -84.8868>}]
 dyn_skirt_side_aa_r [{shapeType=Capsule, length=17.2175, radius=6.3145, center=<-16.1658 3.3601 -0.6965>, rotation=<rot 0 86.8604 -83.2317>}]
 dyn_skirt_side_ba_l [{shapeType=Capsule, length=17.015, radius=7.7879, center=<16.2392 -3.8208 -0.9794>, rotation=<rot 0 85.2348 -83.2753>}]
 dyn_skirt_side_ba_r [{shapeType=Capsule, length=17.3417, radius=8.071, center=<-16.3361 3.1888 1.6284>, rotation=<rot 0 87.0406 -80.1598>}]
 foot_l [{shapeType=Capsule, length=14.3853, radius=8.157, center=<-2.987 8.1962 0.023>, rotation=<rot 0 7.152 -87.0853>}]
 foot_r [{shapeType=Capsule, length=14.3853, radius=8.157, center=<2.9868 -8.1962 -0.023>, rotation=<rot 0 7.152 -87.0853>}]
 hand_l [{shapeType=Capsule, length=8.8998, radius=6.4172, center=<8.0493 -2.911 0.2403>, rotation=<rot 0 87.469 -67.9922>}]
 hand_r [{shapeType=Capsule, length=8.8998, radius=6.4172, center=<-8.0492 2.9111 -0.2403>, rotation=<rot 0 87.469 -67.9922>}]
 lowerarm_l [{shapeType=Capsule, length=14.4656, radius=6.2825, center=<9.6068 -0.1418 0.1349>, rotation=<rot 0 103.369 -87.6901>}]
 lowerarm_r [{shapeType=Capsule, length=14.4656, radius=6.2825, center=<-9.6068 0.1418 -0.1348>, rotation=<rot 0 103.369 -87.6901>}]
 pelvis [{shapeType=Capsule, length=24.7379, radius=26.2088, center=<-25.4665 3.5683 0.0734>, rotation=<rot 0 93.8677 90>}]
 spine_01 [{shapeType=Capsule, length=12.6466, radius=19.4559, center=<0.7992 1.7173 0.0126>, rotation=<rot -0.243 -90 0.001>}]
 spine_02 [{shapeType=Capsule, length=17.5432, radius=14.3625, center=<2.8481 7.5531 0.007>, rotation=<rot -0.0305 -90 0.0178>}]
 spine_03 [{shapeType=Capsule, length=17.257, radius=20.7062, center=<7.769 2.0949 0.483>, rotation=<rot -8.1468 -101.729 -86.0318>}]
 spine_04 [{shapeType=Capsule, length=19.2068, radius=15.1175, center=<9.2082 8.6692 -0.065>, rotation=<rot 0.4264 -90 0.1013>}]
 spine_05 [{shapeType=Capsule, length=7.5224, radius=17.5481, center=<2.1989 -3.6081 0.0212>, rotation=<rot -0.1618 -90 0.4273>}]
 thigh_l [{shapeType=Capsule, length=20.0356, radius=9.3411, center=<-21.5518 1.5147 0.44>, rotation=<rot 0 -91.886 -87.6178>}]
 thigh_r [{shapeType=Capsule, length=20.053, radius=9.3719, center=<21.5197 -1.4145 -0.4223>, rotation=<rot 0 -91.4961 -87.8106>}]
 upperarm_l [{shapeType=Capsule, length=16.0612, radius=9.3484, center=<11.7946 -0.7599 -1.3291>, rotation=<rot 0 -99.1861 -82.2799>}]
 upperarm_r [{shapeType=Capsule, length=16.0498, radius=9.401, center=<-11.7037 0.8422 1.3626>, rotation=<rot 0 -100.1459 -82.9888>}]
}
constraints []
