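# Post-copy animation Blueprint for the M_MED_Banana_Leader player body part, adding secondary physics and driving a mouth material parameter.
# AnimGraph copies the attached parent's pose with curves, constrains dyn_neck between neck_01 and spine_05, then runs RigidBody simulation in root-bone space with damped component motion.
# EventGraph sets the smile mouth frame on a dynamic material each update and on emote begin/end; emote and non-emote paths both use frame 1.
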AnimBlueprint /BRCosmetics/Characters/Player/Male/Medium/Bodies/M_MED_Banana_Leader/Meshes/Parts/M_MED_Banana_Leader_AnimBP
target_skeleton=&/M_MED_Banana_Leader_Skeleton
graphs [
 AnimGraph {nodes [
   n1 = "Constraint - dyn_neck (PP)"!(
    ComponentPose=Local To Component(LocalPose=Copy Pose From Mesh() {bCopyCurves=true, bUseAttachedParent=true})
    Alpha=1
    ConstraintWeights_0=0.4
    ConstraintWeights_1:float=0.6
   ) -> (Pose) {
    alphaBoolBlend {blendOption=Linear}
    boneToModify {boneName=dyn_neck}
    constraintSetup [{targetBone {boneName=neck_01}}, {targetBone {boneName=spine_05}}]
    constraintWeights [1,1]
   }
   n2 = RigidBody(ComponentPose=@n1) {
    bClampLinearTranslationLimitToRefPose=true
    bOverrideWorldGravity=true
    overridePhysicsAsset=&/M_MED_Banana_Leader_Physics
    simulationSpace=BaseBoneSpace
    alphaBoolBlend {blendOption=Linear}
    baseBoneRef {boneName=root}
    componentAppliedLinearAccClamp=<800 800 800>
    componentLinearAccScale=<0.65 0.65 0.65>
    componentLinearVelScale=<0.3 0.3 0.3>
    overrideWorldGravity {z=-980}
   }
   Output Pose(Result=Component To Local(ComponentPose=@n2)) {layerGroup=None}
  ]}
 EventGraph {nodes [
   Event BlueprintUpdateAnimation()
   > n1 = Get Dynamic Material Instance(SkeletalMeshComponent=GetPartSkeletalMeshComponent())
   > n2 = Branch(Condition=Get bIsPlayingEmote())
   > SetScalarParameterValue(self=@n1, ParameterName=Smile Mouth animation Frame, Value=1)
   SetScalarParameterValue(execute=@n2.else, self=@n1, ParameterName=Smile Mouth animation Frame, Value=1)
   n3 = Event OnBeginPlayingEmote() {is=title K2Node_Event}
   n4 = Get Dynamic Material Instance(execute=[@n3.then, @n5.then], SkeletalMeshComponent=GetPartSkeletalMeshComponent())
   > Is Valid(InputObject=@n4)
   > SetScalarParameterValue(self=@n4, ParameterName=Smile Mouth animation Frame, Value=SelectFloat(A=1, B=1, bPickA=Get bIsPlayingEmote()))
   n5 = Event OnEndPlayingEmote() {is=title K2Node_Event}
  ]}
]
variables []
